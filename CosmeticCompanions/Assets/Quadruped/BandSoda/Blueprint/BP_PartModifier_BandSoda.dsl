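; Part modifier for the BandSoda cosmetic companion that drives a material clipping disc at its collar when clothed.
; Post-res-in and customization-finished events read the Outfit and Scale variants to set IsClothed and IsChonky.
; While clothed, Tick feeds the collar mask socket's world transform into the mesh materials' disc parameters, fetching the mesh if invalid.
; The chonky variant adds rotation and position offsets and also sets radius and XY scale.
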(event EventOnDayPhaseChanged
  (params (CurrentDayPhase :type EFortDayPhase%20Enum :default "Morning") (PreviousDayPhase :type EFortDayPhase%20Enum :default "Morning") (bAtCreation :type Boolean)))

(event EventOnNewPlayerWeaponType
  (params (NewType :type EFortItemType%20Enum :default "WorldItem") (OldType :type EFortItemType%20Enum :default "WorldItem"))
  (|CallParentFunction NewType OldType) ; This node is disabled and will not be called.\nDrag off pins to build functionality.
)

(event EventOnWeaponFired)

(event EventOnNewWeapons
  (params (WeaponsAsItems :type Array%20of%20Fort%20Item%20Object%20References)))

(event EventOnCurieStateAttached
  (params (Tag :type Gameplay%20Tag%20Structure%20%28by%20ref%29)))

(event EventOnCurieStateDetached
  (params (Tag :type Gameplay%20Tag%20Structure%20%28by%20ref%29)))

(event EventOnBeginSkydiving
  (|CallParentFunction) ; This node is disabled and will not be called.\nDrag off pins to build functionality.
)

(event EventOnEndSkydiving
  (|CallParentFunction) ; This node is disabled and will not be called.\nDrag off pins to build functionality.
)

(event EventOnEnterVehicle
  (params (VehicleEntered :type Fort%20Vehicle%20Interface%20Interface%20%28by%20ref%29))
  (|CallParentFunction VehicleEntered) ; This node is disabled and will not be called.\nDrag off pins to build functionality.
)

(event EventOnExitVehicle
  (|CallParentFunction) ; This node is disabled and will not be called.\nDrag off pins to build functionality.
)

(event EventOnTargetingOrFirstPersonCameraChanged)

(event EventOnHitPawn
  (params (SelfActor :type Actor%20Object%20Reference) (OtherActor :type Actor%20Object%20Reference) (NormalImpulse :type Vector :default "0, 0, 0") (Hit :type Hit%20Result%20Structure%20%28by%20ref%29)))

(event EventOnPostResIn
  (bind _returnvalue (CosmeticLoadout|GetCosmeticLoadout))
  (Variables|Default|SetIsClothed (!= (Class|FortCosmeticFunctionLibrary|GetActiveVariantForSoftItemAndChannel _returnvalue "/CosmeticCompanions/Assets/Items/Companion_BandSoda.Companion_BandSoda" "(TagName=\"Cosmetics.Variant.Channel.Outfit\")") "(TagName=\"Cosmetics.Variant.Property.Stage1\")"))
  (Variables|Default|SetIsChonky (== (Class|FortCosmeticFunctionLibrary|GetActiveVariantForSoftItemAndChannel _returnvalue "/CosmeticCompanions/Assets/Items/Companion_BandSoda.Companion_BandSoda" "(TagName=\"Cosmetics.Variant.Channel.Immutable.Scale\")") "(TagName=\"Cosmetics.Variant.Property.Stage1\")")))

(event EventOnCharacterCustomizationFinished
  (params (CharacterPartInfo :type Array%20of%20Applied%20Character%20Part%20Info%20Structures))
  (bind _returnvalue (CosmeticLoadout|GetCosmeticLoadout))
  (Variables|Default|SetIsClothed (!= (Class|FortCosmeticFunctionLibrary|GetActiveVariantForSoftItemAndChannel _returnvalue "/CosmeticCompanions/Assets/Items/Companion_BandSoda.Companion_BandSoda" "(TagName=\"Cosmetics.Variant.Channel.Outfit\")") "(TagName=\"Cosmetics.Variant.Property.Stage1\")"))
  (Variables|Default|SetIsChonky (== (Class|FortCosmeticFunctionLibrary|GetActiveVariantForSoftItemAndChannel _returnvalue "/CosmeticCompanions/Assets/Items/Companion_BandSoda.Companion_BandSoda" "(TagName=\"Cosmetics.Variant.Channel.Immutable.Scale\")") "(TagName=\"Cosmetics.Variant.Property.Stage1\")")))

(event EventTick
  (params (DeltaSeconds :type Float%20%28single-precision%29))
  (bind _sockettransform (Variables|Default|GetSocketTransform))
  (|CallParentFunction DeltaSeconds)
  (if (Variables|Default|GetIsClothed)
    (Utilities|IsValid (Variables|Default|GetSkelMeshComponent)
      (:Is%20Valid
        (Variables|Default|SetSocketTransform (Transformation|GetSocketTransform (Variables|Default|GetSkelMeshComponent) "deform_collar_mat_maskSocket" "RTS_World"))
        (if (Variables|Default|GetIsChonky)
          (Rendering|Material|SetVectorParameterValueonMaterials (Variables|Default|GetSkelMeshComponent) "Disc Direction" (Math|Rotator|GetRotationXVector (Math|Rotator|CombineRotators (Variables|Default|GetBoneRotationOffset) (.rotation (Variables|Default|GetSocketTransform)))))
          (Rendering|Material|SetVectorParameterValueonMaterials (Variables|Default|GetSkelMeshComponent) "Disc Pos" (Math|Transform|TransformLocation (Variables|Default|GetSocketTransform) (Variables|Default|GetOffsetPos)))
          (Rendering|Material|SetScalarParameterValueonMaterials (Variables|Default|GetSkelMeshComponent) "Clipping Plane Radius" (Variables|Default|GetRadius))
          (Rendering|Material|SetVectorParameterValueonMaterials (Variables|Default|GetSkelMeshComponent) "XY Scale" "1.000000,1.000000,0.000000")
          (else
            (Rendering|Material|SetVectorParameterValueonMaterials (Variables|Default|GetSkelMeshComponent) "Disc Direction" (Math|Rotator|GetRotationXVector (.rotation _sockettransform)))
            (Rendering|Material|SetVectorParameterValueonMaterials (Variables|Default|GetSkelMeshComponent) "Disc Pos" (.location _sockettransform)))))
      (:Is%20Not%20Valid
        (Variables|Default|SetSkelMeshComponent (CustomCharacter|GetSkeletalMeshforPartonOwner))))))

(fn ConstructionScript
  (|CallParentFunction))
