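# Material for the head of the BandSoda quadruped cosmetic companion, apparently a Slurp liquid look judging by its name.
# Analytic ray-cylinder intersections, camera vectors and a scene depth comparison build masks for a faked inner volume.
# Lighting is computed in-graph from the sunlight collection parameter, time-of-day scalars, Fresnel and custom spherical harmonics evaluation.
# Pre-skinned world-aligned panning textures and vertex colors add detail, and attributes pass through MF_TechArtOverrides.
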Material /CosmeticCompanions/Assets/Quadruped/BandSoda/Materials/M_BandSoda_Slurp_Head
graph {
 expressions {
  MaterialExpressionAbs_0 {class=MaterialExpressionAbs, inputs [None {from=MaterialExpressionVertexInterpolator_5, output=PS}], properties {}}
  MaterialExpressionAbs_1 {class=MaterialExpressionAbs, inputs [None {from=MaterialExpressionDotProduct_1, output=""}], properties {}}
  MaterialExpressionAbs_7 {class=MaterialExpressionAbs, inputs [None {from=MaterialExpressionScalarParameter_49, output=""}], properties {}}
  MaterialExpressionActorPositionWS_0 {class=MaterialExpressionActorPositionWS, inputs [], properties {}}
  MaterialExpressionActorPositionWS_2 {class=MaterialExpressionActorPositionWS, inputs [], properties {}}
  MaterialExpressionAdd_0 {
   class=MaterialExpressionAdd
   inputs [
    A {from=MaterialExpressionShadingPathSwitch_1}
    B {from=MaterialExpressionMultiply_0}
   ]
  }
  MaterialExpressionAdd_1 {
   class=MaterialExpressionAdd
   inputs [
    A {from=MaterialExpressionMultiply_5}
    B {from=""}
   ]
   properties {constB=0.5}
  }
  MaterialExpressionAdd_10 {
   class=MaterialExpressionAdd
   inputs [
    A {from=MaterialExpressionSine_2}
    B {from=""}
   ]
  }
  MaterialExpressionAdd_11 {
   class=MaterialExpressionAdd
   inputs [
    A {from=MaterialExpressionNamedRerouteUsage_10}
    B {from=""}
   ]
  }
  MaterialExpressionAdd_12 {
   class=MaterialExpressionAdd
   inputs [
    A {from=MaterialExpressionNamedRerouteDeclaration_0}
    B {from=MaterialExpressionAdd_14}
   ]
  }
  MaterialExpressionAdd_13 {
   class=MaterialExpressionAdd
   inputs [
    A {from=MaterialExpressionDotProduct_1}
    B {from=""}
   ]
   properties {constB=1.75}
  }
  MaterialExpressionAdd_14 {
   class=MaterialExpressionAdd
   inputs [
    A {from=MaterialExpressionSubtract_8}
    B {from=MaterialExpressionMultiply_24}
   ]
  }
  MaterialExpressionAdd_15 {
   class=MaterialExpressionAdd
   inputs [
    A {from=MaterialExpressionDotProduct_1}
    B {from=MaterialExpressionScalarParameter_10}
   ]
   properties {constB=1.25}
  }
  MaterialExpressionAdd_16 {
   class=MaterialExpressionAdd
   inputs [
    A {from=MaterialExpressionVertexColor_0, output=R}
    B {from=""}
   ]
  }
  MaterialExpressionAdd_17 {
   class=MaterialExpressionAdd
   inputs [
    A {from=MaterialExpressionMultiply_21}
    B {from=""}
   ]
   properties {constB=-0.5}
  }
  MaterialExpressionAdd_18 {
   class=MaterialExpressionAdd
   inputs [
    A {from=MaterialExpressionMultiply_7}
    B {from=MaterialExpressionMultiply_69}
   ]
  }
  MaterialExpressionAdd_19 {
   class=MaterialExpressionAdd
   inputs [
    A {from=MaterialExpressionAbs_0}
    B {from=MaterialExpressionMultiply_8}
   ]
   properties {constB=-0.55}
  }
  MaterialExpressionAdd_2 {
   class=MaterialExpressionAdd
   inputs [
    A {from=MaterialExpressionMultiply_69}
    B {from=MaterialExpressionMultiply_71}
   ]
  }
  MaterialExpressionAdd_20 {
   class=MaterialExpressionAdd
   inputs [
    A {from=MaterialExpressionDotProduct_1}
    B {from=""}
   ]
   properties {constB=0.25}
  }
  MaterialExpressionAdd_21 {
   class=MaterialExpressionAdd
   inputs [
    A {from=MaterialExpressionVertexInterpolator_4, output=PS}
    B {from=MaterialExpressionMultiply_6}
   ]
  }
  MaterialExpressionAdd_22 {
   class=MaterialExpressionAdd
   inputs [
    A {from=MaterialExpressionNamedRerouteDeclaration_1}
    B {from=""}
   ]
   properties {constB=0.6}
  }
  MaterialExpressionAdd_23 {
   class=MaterialExpressionAdd
   inputs [
    A {from=MaterialExpressionSaturate_9}
    B {from=MaterialExpressionSaturate_11}
   ]
  }
  MaterialExpressionAdd_24 {
   class=MaterialExpressionAdd
   inputs [
    A {from=MaterialExpressionMultiply_12}
    B {from=MaterialExpressionMultiply_28}
   ]
  }
  MaterialExpressionAdd_26 {
   class=MaterialExpressionAdd
   inputs [
    A {from=MaterialExpressionMultiply_31}
    B {from=""}
   ]
   properties {constB=0.5}
  }
  MaterialExpressionAdd_27 {
   class=MaterialExpressionAdd
   inputs [
    A {from=MaterialExpressionSaturate_21}
    B {from=""}
   ]
   properties {constB=0.5}
  }
  MaterialExpressionAdd_28 {
   class=MaterialExpressionAdd
   inputs [
    A {from=MaterialExpressionMultiply_37}
    B {from=MaterialExpressionAppendVector_16}
   ]
  }
  MaterialExpressionAdd_29 {
   class=MaterialExpressionAdd
   inputs [
    A {from=MaterialExpressionComponentMask_13}
    B {from=MaterialExpressionMultiply_54}
   ]
  }
  MaterialExpressionAdd_3 {
   class=MaterialExpressionAdd
   inputs [
    A {from=MaterialExpressionReroute_23}
    B {from=MaterialExpressionLinearInterpolate_2}
   ]
  }
  MaterialExpressionAdd_30 {
   class=MaterialExpressionAdd
   inputs [
    A {from=MaterialExpressionNamedRerouteUsage_0}
    B {from=""}
   ]
  }
  MaterialExpressionAdd_31 {
   class=MaterialExpressionAdd
   inputs [
    A {from=MaterialExpressionSaturate_17}
    B {from=MaterialExpressionSaturate_18}
   ]
  }
  MaterialExpressionAdd_32 {
   class=MaterialExpressionAdd
   inputs [
    A {from=MaterialExpressionPower_6}
    B {from=MaterialExpressionScalarParameter_11}
   ]
  }
  MaterialExpressionAdd_33 {
   class=MaterialExpressionAdd
   inputs [
    A {from=MaterialExpressionDotProduct_4}
    B {from=""}
   ]
  }
  MaterialExpressionAdd_34 {
   class=MaterialExpressionAdd
   inputs [
    A {from=MaterialExpressionLinearInterpolate_11}
    B {from=MaterialExpressionLinearInterpolate_12}
   ]
  }
  MaterialExpressionAdd_35 {
   class=MaterialExpressionAdd
   inputs [
    A {from=MaterialExpressionMultiply_90}
    B {from=MaterialExpressionReroute_17}
   ]
  }
  MaterialExpressionAdd_36 {
   class=MaterialExpressionAdd
   inputs [
    A {from=MaterialExpressionNamedRerouteUsage_24}
    B {from=""}
   ]
  }
  MaterialExpressionAdd_37 {
   class=MaterialExpressionAdd
   inputs [
    A {from=MaterialExpressionDotProduct_7}
    B {from=""}
   ]
  }
  MaterialExpressionAdd_38 {
   class=MaterialExpressionAdd
   inputs [
    A {from=MaterialExpressionDotProduct_12}
    B {from=""}
   ]
  }
  MaterialExpressionAdd_39 {
   class=MaterialExpressionAdd
   inputs [
    A {from=MaterialExpressionVectorParameter_14, output=RGB}
    B {from=MaterialExpressionVectorParameter_19, output=RGB}
   ]
  }
  MaterialExpressionAdd_4 {
   class=MaterialExpressionAdd
   inputs [
    A {from=MaterialExpressionReroute_4}
    B {from=MaterialExpressionLinearInterpolate_1}
   ]
  }
  MaterialExpressionAdd_40 {
   class=MaterialExpressionAdd
   inputs [
    A {from=MaterialExpressionOneMinus_13}
    B {from=MaterialExpressionMultiply_49}
   ]
  }
  MaterialExpressionAdd_41 {
   class=MaterialExpressionAdd
   inputs [
    A {from=MaterialExpressionStep_3}
    B {from=MaterialExpressionPower_7}
   ]
  }
  MaterialExpressionAdd_42 {
   class=MaterialExpressionAdd
   inputs [
    A {from=MaterialExpressionVectorParameter_18, output=RGB}
    B {from=MaterialExpressionActorPositionWS_0}
   ]
  }
  MaterialExpressionAdd_43 {
   class=MaterialExpressionAdd
   inputs [
    A {from=MaterialExpressionVectorParameter_5, output=RGB}
    B {from=MaterialExpressionVectorParameter_20, output=RGB}
   ]
  }
  MaterialExpressionAdd_44 {
   class=MaterialExpressionAdd
   inputs [
    A {from=MaterialExpressionScalarParameter_30}
    B {from=MaterialExpressionLinearInterpolate_16}
   ]
  }
  MaterialExpressionAdd_47 {
   class=MaterialExpressionAdd
   inputs [
    A {from=MaterialExpressionVectorParameter_7, output=RGB}
    B {from=MaterialExpressionMultiply_57}
   ]
  }
  MaterialExpressionAdd_48 {
   class=MaterialExpressionAdd
   inputs [
    A {from=MaterialExpressionLinearInterpolate_25}
    B {from=MaterialExpressionDivide_11}
   ]
  }
  MaterialExpressionAdd_49 {
   class=MaterialExpressionAdd
   inputs [
    A {from=MaterialExpressionDotProduct_13}
    B {from=MaterialExpressionScalarParameter_61}
   ]
  }
  MaterialExpressionAdd_5 {
   class=MaterialExpressionAdd
   inputs [
    A {from=MaterialExpressionAdd_1}
    B {from=MaterialExpressionMultiply_11}
   ]
  }
  MaterialExpressionAdd_6 {
   class=MaterialExpressionAdd
   inputs [
    A {from=MaterialExpressionMaterialFunctionCall_0, output=R}
    B {from=MaterialExpressionMaterialFunctionCall_0, output=R}
   ]
  }
  MaterialExpressionAdd_7 {
   class=MaterialExpressionAdd
   inputs [
    A {from=MaterialExpressionMultiply_15}
    B {from=MaterialExpressionVectorParameter_2, output=RGB}
   ]
  }
  MaterialExpressionAdd_8 {
   class=MaterialExpressionAdd
   inputs [
    A {from=MaterialExpressionAdd_6}
    B {from=MaterialExpressionMaterialFunctionCall_0, output=B}
   ]
  }
  MaterialExpressionAdd_9 {
   class=MaterialExpressionAdd
   inputs [
    A {from=MaterialExpressionReroute_7}
    B {from=MaterialExpressionMultiply_79}
   ]
  }
  MaterialExpressionAppendVector_10 {
   class=MaterialExpressionAppendVector
   inputs [
    A {from=MaterialExpressionMultiply_53, output=""}
    B {from=MaterialExpressionScalarParameter_34, output=""}
   ]
  }
  MaterialExpressionAppendVector_15 {
   class=MaterialExpressionAppendVector
   inputs [
    A {from=MaterialExpressionVectorParameter_12, output=RGB}
    B {from=MaterialExpressionVectorParameter_12, output=RGB}
   ]
  }
  MaterialExpressionAppendVector_16 {
   class=MaterialExpressionAppendVector
   inputs [
    A {from=MaterialExpressionVectorParameter_13, output=RGB}
    B {from=MaterialExpressionVectorParameter_13, output=RGB}
   ]
  }
  MaterialExpressionAppendVector_17 {
   class=MaterialExpressionAppendVector
   inputs [
    A {from=MaterialExpressionVertexColor_0, output=""}
    B {from=MaterialExpressionVertexColor_0, output=""}
   ]
  }
  MaterialExpressionAppendVector_5 {
   class=MaterialExpressionAppendVector
   inputs [
    A {from=MaterialExpressionAdd_16, output=""}
    B {from=MaterialExpressionAdd_17, output=""}
   ]
  }
  MaterialExpressionCameraPositionWS_0 {class=MaterialExpressionCameraPositionWS, inputs [], properties {}}
  MaterialExpressionCameraPositionWS_1 {class=MaterialExpressionCameraPositionWS, inputs [], properties {}}
  MaterialExpressionCameraPositionWS_3 {class=MaterialExpressionCameraPositionWS, inputs [], properties {}}
  MaterialExpressionClamp_0 {
   class=MaterialExpressionClamp
   inputs [
    None {from=MaterialExpressionLinearInterpolate_21, output=""}
    Min {from=MaterialExpressionScalarParameter_25, output=""}
    Max {from="", output=""}
   ]
   properties {maxDefault=0}
  }
  MaterialExpressionClamp_1 {
   class=MaterialExpressionClamp
   inputs [
    None {from=MaterialExpressionLinearInterpolate_23, output=""}
    Min {from="", output=""}
    Max {from=MaterialExpressionAdd_44, output=""}
   ]
   properties {}
  }
  MaterialExpressionCollectionParameter_0 {
   class=MaterialExpressionCollectionParameter
   inputs []
   properties {
    collection=&FortniteMaterialParameters
    parameterName=SunlightDirectionalVector
   }
  }
  MaterialExpressionCollectionParameter_3 {
   class=MaterialExpressionCollectionParameter
   inputs []
   properties {
    collection=&FortniteMaterialParameters
    parameterName=SunlightDirectionalVector
   }
  }
  MaterialExpressionComponentMask_0 {class=MaterialExpressionComponentMask, inputs [None {from=MaterialExpressionReroute_3}], properties {b=true, g=true}}
  MaterialExpressionComponentMask_1 {class=MaterialExpressionComponentMask, inputs [None {from=MaterialExpressionCollectionParameter_0}], properties {b=true, g=true, r=true}}
  MaterialExpressionComponentMask_10 {class=MaterialExpressionComponentMask, inputs [None {from=MaterialExpressionMaterialFunctionCall_20, output=Result}], properties {b=true}}
  MaterialExpressionComponentMask_11 {class=MaterialExpressionComponentMask, inputs [None {from=MaterialExpressionMaterialFunctionCall_25, output=Result}], properties {b=true}}
  MaterialExpressionComponentMask_13 {class=MaterialExpressionComponentMask, inputs [None {from=MaterialExpressionAdd_21}], properties {b=true}}
  MaterialExpressionComponentMask_14 {class=MaterialExpressionComponentMask, inputs [None {from=MaterialExpressionReroute_21}], properties {g=true}}
  MaterialExpressionComponentMask_15 {class=MaterialExpressionComponentMask, inputs [None {from=MaterialExpressionReroute_21}], properties {b=true}}
  MaterialExpressionComponentMask_16 {class=MaterialExpressionComponentMask, inputs [None {from=MaterialExpressionCollectionParameter_3}], properties {b=true, g=true, r=true}}
  MaterialExpressionComponentMask_2 {class=MaterialExpressionComponentMask, inputs [None {from=MaterialExpressionReroute_21}], properties {r=true}}
  MaterialExpressionComponentMask_23 {class=MaterialExpressionComponentMask, inputs [None {from=MaterialExpressionReroute_3}], properties {g=true, r=true}}
  MaterialExpressionComponentMask_24 {class=MaterialExpressionComponentMask, inputs [None {from=MaterialExpressionReroute_3}], properties {b=true, r=true}}
  MaterialExpressionComponentMask_3 {class=MaterialExpressionComponentMask, inputs [None {from=MaterialExpressionSaturate_7}], properties {r=true}}
  MaterialExpressionComponentMask_4 {class=MaterialExpressionComponentMask, inputs [None {from=MaterialExpressionMaterialFunctionCall_8, output=UVChannel1}], properties {r=true}}
  MaterialExpressionComponentMask_5 {class=MaterialExpressionComponentMask, inputs [None {from=MaterialExpressionSceneDepth_0, output=R}], properties {r=true}}
  MaterialExpressionComponentMask_6 {class=MaterialExpressionComponentMask, inputs [None {from=MaterialExpressionNamedRerouteUsage_7}], properties {g=true}}
  MaterialExpressionComponentMask_7 {class=MaterialExpressionComponentMask, inputs [None {from=MaterialExpressionMaterialFunctionCall_8, output=UVChannel1}], properties {g=true}}
  MaterialExpressionComponentMask_8 {
   class=MaterialExpressionComponentMask
   inputs [None {from=MaterialExpressionMaterialFunctionCall_13, output=Result}]
   properties {b=true, g=true, r=true}
  }
  MaterialExpressionComponentMask_9 {class=MaterialExpressionComponentMask, inputs [None {from=MaterialExpressionRotateAboutAxis_0}], properties {b=true}}
  MaterialExpressionConstant3Vector_0 {class=MaterialExpressionConstant3Vector}
  MaterialExpressionConstant3Vector_1 {class=MaterialExpressionConstant3Vector, properties {constant=<rgba 1 1 1 0>}}
  MaterialExpressionConstant3Vector_11 {class=MaterialExpressionConstant3Vector, properties {constant=<rgba 1 0 0 0>}}
  MaterialExpressionConstant3Vector_12 {class=MaterialExpressionConstant3Vector, properties {constant=<rgba 0 -1 0 0>}}
  MaterialExpressionConstant3Vector_13 {class=MaterialExpressionConstant3Vector, properties {constant=<rgba 0 10 0 0>}}
  MaterialExpressionConstant3Vector_14 {class=MaterialExpressionConstant3Vector, properties {constant=<rgba 0 -1 0 0>}}
  MaterialExpressionConstant3Vector_2 {class=MaterialExpressionConstant3Vector, properties {constant=<rgba 0 0 -1 0>}}
  MaterialExpressionConstant3Vector_3 {class=MaterialExpressionConstant3Vector, properties {constant=<rgba 0 -1 0 0>}}
  MaterialExpressionConstant3Vector_4 {class=MaterialExpressionConstant3Vector}
  MaterialExpressionConstant3Vector_5 {class=MaterialExpressionConstant3Vector, properties {constant=<rgba 0 -1 0 0>}}
  MaterialExpressionConstant3Vector_6 {class=MaterialExpressionConstant3Vector, properties {constant=<rgba 0 0 1 0>}}
  MaterialExpressionConstant3Vector_7 {class=MaterialExpressionConstant3Vector, properties {constant=<rgba 0 0 1 0>}}
  MaterialExpressionConstant3Vector_9 {class=MaterialExpressionConstant3Vector}
  MaterialExpressionConstant_0 {class=MaterialExpressionConstant, properties {r=4}}
  MaterialExpressionConstant_1 {class=MaterialExpressionConstant, properties {r=0.5}}
  MaterialExpressionConstant_10 {class=MaterialExpressionConstant, properties {r=50}}
  MaterialExpressionConstant_11 {class=MaterialExpressionConstant, properties {r=0.66}}
  MaterialExpressionConstant_12 {class=MaterialExpressionConstant, properties {r=0.66}}
  MaterialExpressionConstant_13 {class=MaterialExpressionConstant, properties {r=5}}
  MaterialExpressionConstant_14 {class=MaterialExpressionConstant, properties {r=50}}
  MaterialExpressionConstant_15 {class=MaterialExpressionConstant}
  MaterialExpressionConstant_16 {class=MaterialExpressionConstant, properties {r=1}}
  MaterialExpressionConstant_2 {class=MaterialExpressionConstant, properties {r=1}}
  MaterialExpressionConstant_3 {class=MaterialExpressionConstant, properties {r=1}}
  MaterialExpressionConstant_4 {class=MaterialExpressionConstant}
  MaterialExpressionConstant_6 {class=MaterialExpressionConstant, properties {r=2}}
  MaterialExpressionConstant_7 {class=MaterialExpressionConstant}
  MaterialExpressionConstant_8 {class=MaterialExpressionConstant, properties {r=1}}
  MaterialExpressionConstant_9 {class=MaterialExpressionConstant, properties {r=0.1}}
  MaterialExpressionConvert_1 {
   class=MaterialExpressionConvert
   inputs [
    None {from=""}
    None {from=""}
    None {from=MaterialExpressionScalarParameter_62}
   ]
   properties {
    nodeName=Make Vector3
    convertInputs [
     {type=Scalar, expressionInput {expression=null, inputName=None, mask=0, maskA=0, maskB=0, maskG=0, maskR=0, outputIndex=0}}
     {type=Scalar, expressionInput {expression=null, inputName=None, mask=0, maskA=0, maskB=0, maskG=0, maskR=0, outputIndex=0}}
     {
      type=Scalar
      expressionInput {
       expression=&self:MaterialExpressionScalarParameter_62
       inputName=None
       mask=0
       maskA=0
       maskB=0
       maskG=0
       maskR=0
       outputIndex=0
      }
     }
    ]
    convertOutputs [{type=Vector3}]
   }
  }
  MaterialExpressionCustom_1 {
   class=MaterialExpressionCustom
   inputs [
    SH {from=MaterialExpressionStaticSwitchParameter_1, output=""}
    Direction {from=MaterialExpressionShadingPathSwitch_0, output=""}
   ]
   properties {
    code="//SH decode,  Direction = LightDir, SH = 2band coefficients. (encoded in Y-up)\r\nDirection = normalize(Direction);\r\nfloat4 shEvalDirection;\r\nshEvalDirection.x = 0.282095;\r\nshEvalDirection.y =0.488603 * Direction.z;\r\nshEvalDirection.z = 0.488603 * Direction.y;\r\nshEvalDirection.w =-0.488603 * Direction.x;\r\nreturn saturate(dot(SH, shEvalDirection));"
    description=SphericalHarmonicsEval
    outputType=CMOT_Float1
    inputs [
     {
      inputName=SH
      input {
       expression=&self:MaterialExpressionStaticSwitchParameter_1
       inputName=None
       mask=0
       maskA=0
       maskB=0
       maskG=0
       maskR=0
       outputIndex=0
      }
     }
     {
      inputName=Direction
      input {
       expression=&self:MaterialExpressionShadingPathSwitch_0
       inputName=None
       mask=0
       maskA=0
       maskB=0
       maskG=0
       maskR=0
       outputIndex=0
      }
     }
    ]
   }
  }
  MaterialExpressionDistance_2 {
   class=MaterialExpressionDistance
   inputs [
    A {from=MaterialExpressionTransform_8, output=""}
    B {from=MaterialExpressionConstant_4, output=""}
   ]
   properties {}
  }
  MaterialExpressionDivide_0 {
   class=MaterialExpressionDivide
   inputs [
    A {from=MaterialExpressionMax_2}
    B {from=MaterialExpressionDotProduct_3}
   ]
  }
  MaterialExpressionDivide_1 {
   class=MaterialExpressionDivide
   inputs [
    A {from=MaterialExpressionReroute_27}
    B {from=MaterialExpressionMultiply_9}
   ]
  }
  MaterialExpressionDivide_11 {
   class=MaterialExpressionDivide
   inputs [
    A {from=MaterialExpressionDotProduct_13}
    B {from=MaterialExpressionScalarParameter_61}
   ]
  }
  MaterialExpressionDivide_2 {
   class=MaterialExpressionDivide
   inputs [
    A {from=MaterialExpressionSubtract_9}
    B {from=MaterialExpressionScalarParameter_20}
   ]
  }
  MaterialExpressionDivide_3 {
   class=MaterialExpressionDivide
   inputs [
    A {from=MaterialExpressionSubtract_14}
    B {from=MaterialExpressionScalarParameter_26}
   ]
  }
  MaterialExpressionDivide_4 {
   class=MaterialExpressionDivide
   inputs [
    A {from=MaterialExpressionSubtract_11}
    B {from=MaterialExpressionScalarParameter_15}
   ]
  }
  MaterialExpressionDivide_5 {
   class=MaterialExpressionDivide
   inputs [
    A {from=MaterialExpressionSubtract_21}
    B {from=MaterialExpressionScalarParameter_45}
   ]
  }
  MaterialExpressionDivide_6 {
   class=MaterialExpressionDivide
   inputs [
    A {from=MaterialExpressionSubtract_22}
    B {from=MaterialExpressionScalarParameter_41}
   ]
  }
  MaterialExpressionDivide_7 {
   class=MaterialExpressionDivide
   inputs [
    A {from=MaterialExpressionSubtract_25}
    B {from=MaterialExpressionSubtract_26}
   ]
  }
  MaterialExpressionDivide_8 {
   class=MaterialExpressionDivide
   inputs [
    A {from=MaterialExpressionSubtract_23}
    B {from=MaterialExpressionScalarParameter_29}
   ]
  }
  MaterialExpressionDotProduct_0 {
   class=MaterialExpressionDotProduct
   inputs [
    A {from=MaterialExpressionComponentMask_1, output=""}
    B {from=MaterialExpressionVertexNormalWS_0, output=""}
   ]
   properties {}
  }
  MaterialExpressionDotProduct_1 {
   class=MaterialExpressionDotProduct
   inputs [
    A {from=MaterialExpressionConstant3Vector_3, output=""}
    B {from=MaterialExpressionTransform_1, output=""}
   ]
   properties {}
  }
  MaterialExpressionDotProduct_10 {
   class=MaterialExpressionDotProduct
   inputs [
    A {from=MaterialExpressionSubtract_24, output=""}
    B {from=MaterialExpressionNormalize_5, output=""}
   ]
   properties {}
  }
  MaterialExpressionDotProduct_11 {
   class=MaterialExpressionDotProduct
   inputs [
    A {from=MaterialExpressionNamedRerouteDeclaration_6, output=""}
    B {from=MaterialExpressionNormalize_6, output=""}
   ]
   properties {}
  }
  MaterialExpressionDotProduct_12 {
   class=MaterialExpressionDotProduct
   inputs [
    A {from=MaterialExpressionConstant3Vector_12, output=""}
    B {from=MaterialExpressionNamedRerouteUsage_27, output=""}
   ]
   properties {}
  }
  MaterialExpressionDotProduct_13 {
   class=MaterialExpressionDotProduct
   inputs [
    A {from=MaterialExpressionSubtract_27, output=""}
    B {from=MaterialExpressionNormalize_9, output=""}
   ]
   properties {}
  }
  MaterialExpressionDotProduct_2 {
   class=MaterialExpressionDotProduct
   inputs [
    A {from=MaterialExpressionNormalize_0, output=""}
    B {from=MaterialExpressionMultiply_3, output=""}
   ]
   properties {}
  }
  MaterialExpressionDotProduct_3 {
   class=MaterialExpressionDotProduct
   inputs [
    A {from=MaterialExpressionMax_2, output=""}
    B {from=MaterialExpressionConstant3Vector_1, output=""}
   ]
   properties {}
  }
  MaterialExpressionDotProduct_4 {
   class=MaterialExpressionDotProduct
   inputs [
    A {from=MaterialExpressionConstant3Vector_5, output=""}
    B {from=MaterialExpressionNamedRerouteUsage_6, output=""}
   ]
   properties {}
  }
  MaterialExpressionDotProduct_5 {
   class=MaterialExpressionDotProduct
   inputs [
    A {from=MaterialExpressionConstant3Vector_2, output=""}
    B {from=MaterialExpressionTransform_1, output=""}
   ]
   properties {}
  }
  MaterialExpressionDotProduct_6 {
   class=MaterialExpressionDotProduct
   inputs [
    A {from=MaterialExpressionSubtract_4, output=""}
    B {from=MaterialExpressionConstant3Vector_6, output=""}
   ]
   properties {}
  }
  MaterialExpressionDotProduct_7 {
   class=MaterialExpressionDotProduct
   inputs [
    A {from=MaterialExpressionNormalize_4, output=""}
    B {from=MaterialExpressionNamedRerouteDeclaration_18, output=""}
   ]
   properties {}
  }
  MaterialExpressionDotProduct_8 {
   class=MaterialExpressionDotProduct
   inputs [
    A {from=MaterialExpressionComponentMask_16, output=""}
    B {from=MaterialExpressionReroute_15, output=""}
   ]
   properties {}
  }
  MaterialExpressionDotProduct_9 {
   class=MaterialExpressionDotProduct
   inputs [
    A {from=MaterialExpressionNormalize_3, output=""}
    B {from=MaterialExpressionMultiply_76, output=""}
   ]
   properties {}
  }
  MaterialExpressionFrac_0 {class=MaterialExpressionFrac, inputs [None {from=MaterialExpressionMultiply_64, output=""}], properties {}}
  MaterialExpressionFrac_1 {class=MaterialExpressionFrac, inputs [None {from=MaterialExpressionMultiply_50, output=""}], properties {}}
  MaterialExpressionIf_0 {
   class=MaterialExpressionIf
   inputs [
    A {from=MaterialExpressionAdd_12, output=""}
    B {from=MaterialExpressionComponentMask_5, output=""}
    "A > B" {from=MaterialExpressionConstant_7, output=""}
    "A == B" {from=MaterialExpressionConstant_8, output=""}
    "A < B" {from=MaterialExpressionConstant_8, output=""}
   ]
   properties {}
  }
  MaterialExpressionLength_0 {class=MaterialExpressionLength, inputs [None {from=MaterialExpressionSubtract_10, output=""}], properties {}}
  MaterialExpressionLength_1 {class=MaterialExpressionLength, inputs [None {from=MaterialExpressionNamedRerouteDeclaration_12, output=""}], properties {}}
  MaterialExpressionLinearInterpolate_0 {
   class=MaterialExpressionLinearInterpolate
   inputs [
    A {from=MaterialExpressionLinearInterpolate_7}
    B {from=MaterialExpressionVectorParameter_1, output=RGB}
    Alpha {from=MaterialExpressionSaturate_6}
   ]
  }
  MaterialExpressionLinearInterpolate_1 {
   class=MaterialExpressionLinearInterpolate
   inputs [
    A {from=MaterialExpressionVectorParameter_3, output=RGB}
    B {from=MaterialExpressionAdd_7}
    Alpha {from=MaterialExpressionMin_1}
   ]
  }
  MaterialExpressionLinearInterpolate_10 {
   class=MaterialExpressionLinearInterpolate
   inputs [
    A {from=""}
    B {from=MaterialExpressionMultiply_45}
    Alpha {from=MaterialExpressionNamedRerouteUsage_1}
   ]
   properties {constA=1}
  }
  MaterialExpressionLinearInterpolate_11 {
   class=MaterialExpressionLinearInterpolate
   inputs [
    A {from=MaterialExpressionScalarParameter_17}
    B {from=MaterialExpressionScalarParameter_18}
    Alpha {from=MaterialExpressionPower_0}
   ]
  }
  MaterialExpressionLinearInterpolate_12 {
   class=MaterialExpressionLinearInterpolate
   inputs [
    A {from=""}
    B {from=MaterialExpressionScalarParameter_24}
    Alpha {from=MaterialExpressionPower_3}
   ]
  }
  MaterialExpressionLinearInterpolate_13 {
   class=MaterialExpressionLinearInterpolate
   inputs [
    A {from=""}
    B {from=""}
    Alpha {from=MaterialExpressionNamedRerouteUsage_9}
   ]
   properties {constA=0.5}
  }
  MaterialExpressionLinearInterpolate_14 {
   class=MaterialExpressionLinearInterpolate
   inputs [
    A {from=MaterialExpressionScalarParameter_36}
    B {from=MaterialExpressionScalarParameter_19}
    Alpha {from=MaterialExpressionPower_14}
   ]
  }
  MaterialExpressionLinearInterpolate_15 {
   class=MaterialExpressionLinearInterpolate
   inputs [
    A {from=MaterialExpressionNamedRerouteUsage_4}
    B {from=""}
    Alpha {from=MaterialExpressionNamedRerouteUsage_8}
   ]
   properties {constA=1}
  }
  MaterialExpressionLinearInterpolate_16 {
   class=MaterialExpressionLinearInterpolate
   inputs [
    A {from=""}
    B {from=MaterialExpressionScalarParameter_35}
    Alpha {from=MaterialExpressionMultiply_49}
   ]
  }
  MaterialExpressionLinearInterpolate_17 {
   class=MaterialExpressionLinearInterpolate
   inputs [
    A {from=MaterialExpressionVectorParameter_0, output=RGB}
    B {from=MaterialExpressionAdd_0}
    Alpha {from=MaterialExpressionNamedRerouteDeclaration_3}
   ]
  }
  MaterialExpressionLinearInterpolate_18 {
   class=MaterialExpressionLinearInterpolate
   inputs [
    A {from=MaterialExpressionNamedRerouteUsage_20}
    B {from=""}
    Alpha {from=MaterialExpressionScalarParameter_39}
   ]
  }
  MaterialExpressionLinearInterpolate_19 {
   class=MaterialExpressionLinearInterpolate
   inputs [
    A {from=MaterialExpressionScalarParameter_42}
    B {from=MaterialExpressionScalarParameter_43}
    Alpha {from=MaterialExpressionPower_9}
   ]
  }
  MaterialExpressionLinearInterpolate_2 {
   class=MaterialExpressionLinearInterpolate
   inputs [
    A {from=MaterialExpressionConstant3Vector_4}
    B {from=MaterialExpressionAdd_2}
    Alpha {from=MaterialExpressionSaturate_10}
   ]
  }
  MaterialExpressionLinearInterpolate_20 {
   class=MaterialExpressionLinearInterpolate
   inputs [
    A {from=MaterialExpressionMultiply_41}
    B {from=MaterialExpressionOneMinus_12}
    Alpha {from=MaterialExpressionScalarParameter_46}
   ]
  }
  MaterialExpressionLinearInterpolate_21 {
   class=MaterialExpressionLinearInterpolate
   inputs [
    A {from=""}
    B {from=MaterialExpressionScalarParameter_25}
    Alpha {from=MaterialExpressionPower_8}
   ]
  }
  MaterialExpressionLinearInterpolate_22 {
   class=MaterialExpressionLinearInterpolate
   inputs [
    A {from=MaterialExpressionMaterialFunctionCall_15, output=Opacity Mask}
    B {from=""}
    Alpha {from=MaterialExpressionMultiply_43}
   ]
   properties {constB=0}
  }
  MaterialExpressionLinearInterpolate_23 {
   class=MaterialExpressionLinearInterpolate
   inputs [
    A {from=""}
    B {from=MaterialExpressionAdd_44}
    Alpha {from=MaterialExpressionPower_10}
   ]
  }
  MaterialExpressionLinearInterpolate_24 {
   class=MaterialExpressionLinearInterpolate
   inputs [
    A {from=MaterialExpressionScalarParameter_36}
    B {from=MaterialExpressionScalarParameter_19}
    Alpha {from=MaterialExpressionNamedRerouteUsage_22}
   ]
  }
  MaterialExpressionLinearInterpolate_25 {
   class=MaterialExpressionLinearInterpolate
   inputs [
    A {from=MaterialExpressionMin_3}
    B {from=MaterialExpressionMax_5}
    Alpha {from=MaterialExpressionMultiply_52}
   ]
  }
  MaterialExpressionLinearInterpolate_3 {
   class=MaterialExpressionLinearInterpolate
   inputs [
    A {from=""}
    B {from=MaterialExpressionLinearInterpolate_13}
    Alpha {from=MaterialExpressionNamedRerouteUsage_5}
   ]
   properties {constA=1}
  }
  MaterialExpressionLinearInterpolate_4 {
   class=MaterialExpressionLinearInterpolate
   inputs [
    A {from=MaterialExpressionLinearInterpolate_17}
    B {from=MaterialExpressionLinearInterpolate_6}
    Alpha {from=MaterialExpressionReroute_11}
   ]
  }
  MaterialExpressionLinearInterpolate_5 {
   class=MaterialExpressionLinearInterpolate
   inputs [
    A {from=MaterialExpressionSaturate_15}
    B {from=MaterialExpressionVectorParameter_3, output=RGB}
    Alpha {from=MaterialExpressionMin_2}
   ]
  }
  MaterialExpressionLinearInterpolate_6 {
   class=MaterialExpressionLinearInterpolate
   inputs [
    A {from=MaterialExpressionVectorParameter_0, output=RGB}
    B {from=MaterialExpressionLinearInterpolate_5}
    Alpha {from=MaterialExpressionNamedRerouteDeclaration_3}
   ]
  }
  MaterialExpressionLinearInterpolate_7 {
   class=MaterialExpressionLinearInterpolate
   inputs [
    A {from=MaterialExpressionVectorParameter_9, output=RGB}
    B {from=MaterialExpressionVectorParameter_8, output=RGB}
    Alpha {from=MaterialExpressionSaturate_19}
   ]
  }
  MaterialExpressionLinearInterpolate_8 {
   class=MaterialExpressionLinearInterpolate
   inputs [
    A {from=MaterialExpressionReroute_14}
    B {from=MaterialExpressionSubtract_13}
    Alpha {from=MaterialExpressionScalarParameter_4}
   ]
  }
  MaterialExpressionLinearInterpolate_9 {
   class=MaterialExpressionLinearInterpolate
   inputs [
    A {from=MaterialExpressionComponentMask_8}
    B {from=MaterialExpressionTransform_2}
    Alpha {from=MaterialExpressionMultiply_47}
   ]
  }
  MaterialExpressionLocalPosition_2 {class=MaterialExpressionLocalPosition, inputs [], properties {localOrigin=InstancePreSkinning}}
  MaterialExpressionMaterialFunctionCall_0 {
   class=MaterialExpressionMaterialFunctionCall
   inputs [Float3 {from=MaterialExpressionMultiply_62}]
   properties {materialFunction=&BreakOutFloat3Components}
  }
  MaterialExpressionMaterialFunctionCall_1 {
   class=MaterialExpressionMaterialFunctionCall
   inputs [
    X {from=MaterialExpressionTextureSample_1, output=R}
    Y {from=""}
    Z {from=MaterialExpressionTextureSample_1, output=R}
   ]
   properties {materialFunction=&MakeFloat3}
  }
  MaterialExpressionMaterialFunctionCall_10 {
   class=MaterialExpressionMaterialFunctionCall
   inputs []
   properties {materialFunction=/Engine/Functions/Engine_MaterialFunctions02/Texturing/3DSandMayaUVCoordinates}
  }
  MaterialExpressionMaterialFunctionCall_11 {
   class=MaterialExpressionMaterialFunctionCall
   inputs [
    Value_Morning {from=MaterialExpressionConstant_11}
    Value_Day {from=MaterialExpressionConstant_2}
    Value_Evening {from=MaterialExpressionConstant_12}
    Value_Night {from=MaterialExpressionConstant_1}
   ]
   properties {materialFunction=&Scalar_Time_of_Day}
  }
  MaterialExpressionMaterialFunctionCall_12 {
   class=MaterialExpressionMaterialFunctionCall
   inputs [
    X {from=MaterialExpressionVertexColor_0, output=G}
    Y {from=MaterialExpressionVertexColor_0, output=G}
    Z {from=MaterialExpressionVertexColor_0, output=G}
   ]
   properties {materialFunction=&MakeFloat3}
  }
  MaterialExpressionMaterialFunctionCall_13 {
   class=MaterialExpressionMaterialFunctionCall
   inputs []
   properties {materialFunction=&CameraDirectionVector}
  }
  MaterialExpressionMaterialFunctionCall_14 {
   class=MaterialExpressionMaterialFunctionCall
   inputs [In {from=MaterialExpressionSetMaterialAttributes_0}]
   properties {materialFunction=&MF_TechArtOverrides}
  }
  MaterialExpressionMaterialFunctionCall_15 {
   class=MaterialExpressionMaterialFunctionCall
   inputs [
    Origin {from=MaterialExpressionNamedRerouteDeclaration_16}
    Radius {from=MaterialExpressionLinearInterpolate_19}
    Z Axis {from=MaterialExpressionNamedRerouteDeclaration_18}
    XY Scale {from=MaterialExpressionVectorParameter_10, output=RGB}
    X Axis {from=""}
    Capped? {from=MaterialExpressionStaticBool_1}
    Z Min-Max {from=MaterialExpressionConstant3Vector_9}
   ]
   properties {materialFunction=/Engine/Functions/Engine_MaterialFunctions01/Math/CylinderIntersection}
  }
  MaterialExpressionMaterialFunctionCall_16 {
   class=MaterialExpressionMaterialFunctionCall
   inputs [
    VectorToRotate {from=""}
    LookAtVector {from=""}
    RestingVector {from=""}
   ]
   properties {materialFunction=/Engine/Functions/Engine_MaterialFunctions01/Vectors/RotateVector}
  }
  MaterialExpressionMaterialFunctionCall_17 {
   class=MaterialExpressionMaterialFunctionCall
   inputs [
    X {from=""}
    Y {from=MaterialExpressionTextureSample_8, output=R}
    Z {from=MaterialExpressionTextureSample_8, output=R}
   ]
   properties {materialFunction=&MakeFloat3}
  }
  MaterialExpressionMaterialFunctionCall_19 {
   class=MaterialExpressionMaterialFunctionCall
   inputs [
    TextureObject {from=MaterialExpressionTextureObject_0}
    TextureSize {from=MaterialExpressionScalarParameter_49}
    WorldPosition {from=MaterialExpressionAdd_21}
    Export Float 4 {from=""}
    In {from=MaterialExpressionAppendVector_10}
    World Space Normal {from=MaterialExpressionReroute_30}
    ProjectionTransitionContrast {from=MaterialExpressionConstant_6}
   ]
   properties {materialFunction=/Game/Athena/SafeZone/Preskinned_WP_Texture_ScalePan}
  }
  MaterialExpressionMaterialFunctionCall_2 {
   class=MaterialExpressionMaterialFunctionCall
   inputs []
   properties {materialFunction=&CameraDirectionVector}
  }
  MaterialExpressionMaterialFunctionCall_20 {
   class=MaterialExpressionMaterialFunctionCall
   inputs [
    Vector to Transform {from=MaterialExpressionWorldPosition_3, output=XYZ}
    Z Vector {from=MaterialExpressionNamedRerouteDeclaration_21}
    Center Location {from=MaterialExpressionNamedRerouteUsage_29}
   ]
   properties {materialFunction=/Engine/Functions/Engine_MaterialFunctions02/Math/TransformToZVector}
  }
  MaterialExpressionMaterialFunctionCall_21 {
   class=MaterialExpressionMaterialFunctionCall
   inputs [
    Origin {from=MaterialExpressionNamedRerouteDeclaration_16}
    Radius {from=MaterialExpressionLinearInterpolate_19}
    Z Axis {from=MaterialExpressionNamedRerouteDeclaration_18}
    XY Scale {from=MaterialExpressionVectorParameter_15, output=RGB}
    X Axis {from=""}
    Capped? {from=MaterialExpressionStaticBool_3}
    Z Min-Max {from=MaterialExpressionVectorParameter_16, output=RGB}
   ]
   properties {materialFunction=&CylinderIntersection}
  }
  MaterialExpressionMaterialFunctionCall_24 {
   class=MaterialExpressionMaterialFunctionCall
   inputs [
    Origin {from=MaterialExpressionNamedRerouteUsage_30}
    Radius {from=MaterialExpressionNamedRerouteUsage_32}
    Z Axis {from=MaterialExpressionNamedRerouteDeclaration_21}
    XY Scale {from=MaterialExpressionVectorParameter_21, output=RGB}
    X Axis {from=""}
    Capped? {from=MaterialExpressionStaticBool_2}
    Z Min-Max {from=MaterialExpressionVectorParameter_16, output=RGB}
   ]
   properties {materialFunction=&CylinderIntersection}
  }
  MaterialExpressionMaterialFunctionCall_25 {
   class=MaterialExpressionMaterialFunctionCall
   inputs [
    Vector to Transform {from=MaterialExpressionCameraPositionWS_1}
    Z Vector {from=MaterialExpressionNamedRerouteDeclaration_21}
    Center Location {from=MaterialExpressionNamedRerouteUsage_31}
   ]
   properties {materialFunction=&TransformToZVector}
  }
  MaterialExpressionMaterialFunctionCall_26 {
   class=MaterialExpressionMaterialFunctionCall
   inputs [Float3 {from=MaterialExpressionMaterialFunctionCall_19, output=XY Texture}]
   properties {materialFunction=&BreakOutFloat3Components}
  }
  MaterialExpressionMaterialFunctionCall_28 {
   class=MaterialExpressionMaterialFunctionCall
   inputs [
    TextureObject {from=MaterialExpressionTextureObject_4}
    TextureSize {from=MaterialExpressionScalarParameter_49}
    WorldPosition {from=MaterialExpressionAdd_21}
    Export Float 4 {from=""}
    In {from=MaterialExpressionAppendVector_10}
    World Space Normal {from=MaterialExpressionReroute_30}
    ProjectionTransitionContrast {from=MaterialExpressionConstant_6}
   ]
   properties {materialFunction=&Preskinned_WP_Texture_ScalePan}
  }
  MaterialExpressionMaterialFunctionCall_29 {
   class=MaterialExpressionMaterialFunctionCall
   inputs [
    Normal Vector {from=""}
    Camera Vector {from=""}
    Invert Fresnel {from=""}
    Power {from=""}
    Use Cheap contrast {from=""}
    Cheap contrast dark {from=""}
    Cheap contrast bright {from=""}
    Clamp Fresnel Dot Product {from=""}
   ]
   properties {materialFunction=&Fresnel_Function}
  }
  MaterialExpressionMaterialFunctionCall_3 {
   class=MaterialExpressionMaterialFunctionCall
   inputs [
    Normal Vector {from=""}
    Camera Vector {from=""}
    Invert Fresnel {from=""}
    Power {from=""}
    Use Cheap contrast {from=MaterialExpressionStaticBool_0}
    Cheap contrast dark {from=MaterialExpressionScalarParameter_7}
    Cheap contrast bright {from=MaterialExpressionScalarParameter_6}
    Clamp Fresnel Dot Product {from=""}
   ]
   properties {materialFunction=&Fresnel_Function}
  }
  MaterialExpressionMaterialFunctionCall_30 {
   class=MaterialExpressionMaterialFunctionCall
   inputs [
    WorldPosition {from=MaterialExpressionAdd_21}
    World Space Normal {from=MaterialExpressionReroute_30}
    UVSize_Front {from=MaterialExpressionScalarParameter_49}
    UVSize_Side {from=MaterialExpressionScalarParameter_49}
    UVSize_Top {from=MaterialExpressionScalarParameter_49}
    ProjectionTransitionContrast {from=MaterialExpressionConstant_9}
   ]
   properties {materialFunction=/Game/Packages/Fortress_SharedMaterials/Base_Material_Functions/WorldAlignedTexture_UVsMasks}
  }
  MaterialExpressionMaterialFunctionCall_32 {
   class=MaterialExpressionMaterialFunctionCall
   inputs [
    X {from=MaterialExpressionVertexColor_0, output=R}
    Y {from=MaterialExpressionVertexColor_0, output=R}
    Z {from=MaterialExpressionVertexColor_0, output=R}
    A {from=MaterialExpressionVertexColor_0, output=R}
   ]
   properties {materialFunction=&MakeFloat4}
  }
  MaterialExpressionMaterialFunctionCall_4 {
   class=MaterialExpressionMaterialFunctionCall
   inputs [
    X {from=MaterialExpressionTextureSample_0, output=R}
    Y {from=MaterialExpressionTextureSample_0, output=R}
    Z {from=""}
   ]
   properties {materialFunction=&MakeFloat3}
  }
  MaterialExpressionMaterialFunctionCall_41 {
   class=MaterialExpressionMaterialFunctionCall
   inputs []
   properties {materialFunction=&CameraDirectionVector}
  }
  MaterialExpressionMaterialFunctionCall_46 {
   class=MaterialExpressionMaterialFunctionCall
   inputs [
    Input {from=MaterialExpressionCustom_1}
    Input Low {from=MaterialExpressionScalarParameter_52}
    Input High {from=MaterialExpressionScalarParameter_53}
    Target Low {from=MaterialExpressionConstant_15}
    Target High {from=MaterialExpressionConstant_16}
   ]
   properties {materialFunction=&RemapValueRange}
  }
  MaterialExpressionMaterialFunctionCall_5 {
   class=MaterialExpressionMaterialFunctionCall
   inputs [Material {from=MaterialExpressionMaterialFunctionCall_6, output=Result}]
   properties {materialFunction=/Game/Packages/Fortress_SharedMaterials/Base_Material_Functions/BaseCharacterShaderFunctions/BaseFN_CharacterOpacity}
  }
  MaterialExpressionMaterialFunctionCall_6 {
   class=MaterialExpressionMaterialFunctionCall
   inputs [
    Fresnel {from=""}
    Material {from=MaterialExpressionMaterialFunctionCall_14, output=Result}
   ]
   properties {materialFunction=/Game/Packages/Fortress_SharedMaterials/Base_Material_Functions/CharacterDeRez_Transition}
  }
  MaterialExpressionMaterialFunctionCall_7 {
   class=MaterialExpressionMaterialFunctionCall
   inputs [World Position {from=MaterialExpressionWorldPosition_1, output=XYZ}]
   properties {materialFunction=/Engine/Functions/Engine_MaterialFunctions02/Utility/DepthFromWorldPosition}
  }
  MaterialExpressionMaterialFunctionCall_8 {
   class=MaterialExpressionMaterialFunctionCall
   inputs []
   properties {materialFunction=&3DSandMayaUVCoordinates}
  }
  MaterialExpressionMaterialFunctionCall_9 {
   class=MaterialExpressionMaterialFunctionCall
   inputs []
   properties {materialFunction=&CameraDirectionVector}
  }
  MaterialExpressionMax_0 {
   class=MaterialExpressionMax
   inputs [
    A {from=MaterialExpressionDotProduct_2}
    B {from=""}
   ]
   properties {constB=0}
  }
  MaterialExpressionMax_1 {
   class=MaterialExpressionMax
   inputs [
    A {from=MaterialExpressionDotProduct_9}
    B {from=""}
   ]
   properties {constB=0}
  }
  MaterialExpressionMax_2 {
   class=MaterialExpressionMax
   inputs [
    A {from=MaterialExpressionAdd_19}
    B {from=""}
   ]
   properties {constB=0}
  }
  MaterialExpressionMax_3 {
   class=MaterialExpressionMax
   inputs [
    A {from=MaterialExpressionDotProduct_2}
    B {from=""}
   ]
   properties {constB=0}
  }
  MaterialExpressionMax_5 {
   class=MaterialExpressionMax
   inputs [
    A {from=MaterialExpressionScalarParameter_56}
    B {from=MaterialExpressionMaterialFunctionCall_24, output=Opacity Mask}
   ]
   properties {constA=1}
  }
  MaterialExpressionMin_1 {
   class=MaterialExpressionMin
   inputs [
    A {from=MaterialExpressionSaturate_16, output=""}
    B {from="", output=""}
   ]
   properties {constB=0.7}
  }
  MaterialExpressionMin_2 {
   class=MaterialExpressionMin
   inputs [
    A {from=MaterialExpressionReroute_31, output=""}
    B {from="", output=""}
   ]
   properties {constB=0.95}
  }
  MaterialExpressionMin_3 {
   class=MaterialExpressionMin
   inputs [
    A {from=MaterialExpressionSaturate_37, output=""}
    B {from=MaterialExpressionOneMinus_17, output=""}
   ]
   properties {}
  }
  MaterialExpressionMultiply_0 {
   class=MaterialExpressionMultiply
   inputs [
    A {from=MaterialExpressionReroute_5}
    B {from=MaterialExpressionSaturate_0}
   ]
  }
  MaterialExpressionMultiply_1 {
   class=MaterialExpressionMultiply
   inputs [
    A {from=MaterialExpressionDotProduct_0}
    B {from=""}
   ]
   properties {constB=2}
  }
  MaterialExpressionMultiply_10 {
   class=MaterialExpressionMultiply
   inputs [
    A {from=MaterialExpressionNormalize_1}
    B {from=""}
   ]
   properties {constB=-1}
  }
  MaterialExpressionMultiply_11 {
   class=MaterialExpressionMultiply
   inputs [
    A {from=MaterialExpressionOneMinus_2}
    B {from=""}
   ]
   properties {constB=10}
  }
  MaterialExpressionMultiply_12 {
   class=MaterialExpressionMultiply
   inputs [
    A {from=MaterialExpressionSaturate_2}
    B {from=MaterialExpressionMultiply_46}
   ]
  }
  MaterialExpressionMultiply_13 {
   class=MaterialExpressionMultiply
   inputs [
    A {from=MaterialExpressionSaturate_1}
    B {from=""}
   ]
   properties {constB=10}
  }
  MaterialExpressionMultiply_14 {
   class=MaterialExpressionMultiply
   inputs [
    A {from=MaterialExpressionPower_2}
    B {from=""}
   ]
   properties {constB=30}
  }
  MaterialExpressionMultiply_15 {
   class=MaterialExpressionMultiply
   inputs [
    A {from=MaterialExpressionSaturate_12}
    B {from=""}
   ]
   properties {constB=4}
  }
  MaterialExpressionMultiply_16 {
   class=MaterialExpressionMultiply
   inputs [
    A {from=MaterialExpressionLinearInterpolate_4}
    B {from=MaterialExpressionMaterialFunctionCall_11, output=Result}
   ]
  }
  MaterialExpressionMultiply_17 {
   class=MaterialExpressionMultiply
   inputs [
    A {from=""}
    B {from=""}
   ]
   properties {constB=2}
  }
  MaterialExpressionMultiply_18 {
   class=MaterialExpressionMultiply
   inputs [
    A {from=MaterialExpressionSaturate_4}
    B {from=""}
   ]
   properties {constB=10}
  }
  MaterialExpressionMultiply_19 {
   class=MaterialExpressionMultiply
   inputs [
    A {from=MaterialExpressionMultiply_63}
    B {from=MaterialExpressionComponentMask_3}
   ]
  }
  MaterialExpressionMultiply_2 {
   class=MaterialExpressionMultiply
   inputs [
    A {from=MaterialExpressionMultiply_1}
    B {from=MaterialExpressionVertexNormalWS_0}
   ]
  }
  MaterialExpressionMultiply_20 {
   class=MaterialExpressionMultiply
   inputs [
    A {from=MaterialExpressionOneMinus_0}
    B {from=MaterialExpressionMaterialFunctionCall_21, output=Opacity Mask}
   ]
   properties {constA=1}
  }
  MaterialExpressionMultiply_21 {
   class=MaterialExpressionMultiply
   inputs [
    A {from=MaterialExpressionMaterialFunctionCall_12, output=Result}
    B {from=""}
   ]
  }
  MaterialExpressionMultiply_22 {
   class=MaterialExpressionMultiply
   inputs [
    A {from=MaterialExpressionMultiply_18}
    B {from=MaterialExpressionSaturate_24}
   ]
  }
  MaterialExpressionMultiply_23 {
   class=MaterialExpressionMultiply
   inputs [
    A {from=MaterialExpressionSubtract_5}
    B {from=""}
   ]
   properties {constB=10000}
  }
  MaterialExpressionMultiply_24 {
   class=MaterialExpressionMultiply
   inputs [
    A {from=MaterialExpressionReroute_8}
    B {from=""}
   ]
   properties {constB=5}
  }
  MaterialExpressionMultiply_25 {
   class=MaterialExpressionMultiply
   inputs [
    A {from=MaterialExpressionSaturate_20}
    B {from=MaterialExpressionMultiply_31}
   ]
  }
  MaterialExpressionMultiply_26 {
   class=MaterialExpressionMultiply
   inputs [
    A {from=MaterialExpressionAdd_22}
    B {from=""}
   ]
   properties {constB=-1}
  }
  MaterialExpressionMultiply_27 {
   class=MaterialExpressionMultiply
   inputs [
    A {from=MaterialExpressionSaturate_22}
    B {from=""}
   ]
   properties {constB=2.5}
  }
  MaterialExpressionMultiply_28 {
   class=MaterialExpressionMultiply
   inputs [
    A {from=MaterialExpressionMultiply_32}
    B {from=""}
   ]
   properties {constB=4}
  }
  MaterialExpressionMultiply_29 {
   class=MaterialExpressionMultiply
   inputs [
    A {from=MaterialExpressionAdd_20}
    B {from=""}
   ]
   properties {constB=-1}
  }
  MaterialExpressionMultiply_3 {
   class=MaterialExpressionMultiply
   inputs [
    A {from=MaterialExpressionMaterialFunctionCall_2, output=Result}
    B {from=""}
   ]
   properties {constB=-1}
  }
  MaterialExpressionMultiply_30 {
   class=MaterialExpressionMultiply
   inputs [
    A {from=MaterialExpressionOneMinus_9}
    B {from=MaterialExpressionAdd_23}
   ]
  }
  MaterialExpressionMultiply_31 {
   class=MaterialExpressionMultiply
   inputs [
    A {from=MaterialExpressionSaturate_21}
    B {from=""}
   ]
   properties {constB=1.333}
  }
  MaterialExpressionMultiply_32 {
   class=MaterialExpressionMultiply
   inputs [
    A {from=MaterialExpressionMultiply_27}
    B {from=MaterialExpressionSaturate_23}
   ]
  }
  MaterialExpressionMultiply_33 {
   class=MaterialExpressionMultiply
   inputs [
    A {from=MaterialExpressionNamedRerouteDeclaration_10}
    B {from=MaterialExpressionNormalize_6}
   ]
  }
  MaterialExpressionMultiply_34 {
   class=MaterialExpressionMultiply
   inputs [
    A {from=MaterialExpressionSaturate_34}
    B {from=MaterialExpressionOneMinus_11}
   ]
  }
  MaterialExpressionMultiply_35 {
   class=MaterialExpressionMultiply
   inputs [
    A {from=MaterialExpressionAdd_30}
    B {from=""}
   ]
   properties {constB=0.5}
  }
  MaterialExpressionMultiply_36 {
   class=MaterialExpressionMultiply
   inputs [
    A {from=MaterialExpressionSaturate_3}
    B {from=MaterialExpressionLinearInterpolate_10}
   ]
  }
  MaterialExpressionMultiply_37 {
   class=MaterialExpressionMultiply
   inputs [
    A {from=MaterialExpressionAppendVector_17}
    B {from=MaterialExpressionSubtract_3}
   ]
  }
  MaterialExpressionMultiply_38 {
   class=MaterialExpressionMultiply
   inputs [
    A {from=MaterialExpressionAdd_33}
    B {from=""}
   ]
   properties {constB=0.5}
  }
  MaterialExpressionMultiply_39 {
   class=MaterialExpressionMultiply
   inputs [
    A {from=MaterialExpressionMultiply_36}
    B {from=MaterialExpressionLinearInterpolate_3}
   ]
  }
  MaterialExpressionMultiply_4 {
   class=MaterialExpressionMultiply
   inputs [
    A {from=MaterialExpressionPower_1}
    B {from=MaterialExpressionScalarParameter_3}
   ]
   properties {constB=0.2}
  }
  MaterialExpressionMultiply_40 {
   class=MaterialExpressionMultiply
   inputs [
    A {from=MaterialExpressionDotProduct_6}
    B {from=MaterialExpressionReroute_0}
   ]
  }
  MaterialExpressionMultiply_41 {
   class=MaterialExpressionMultiply
   inputs [
    A {from=MaterialExpressionLinearInterpolate_18}
    B {from=MaterialExpressionNamedRerouteUsage_21}
   ]
  }
  MaterialExpressionMultiply_42 {
   class=MaterialExpressionMultiply
   inputs [
    A {from=MaterialExpressionAdd_36}
    B {from=""}
   ]
   properties {constB=0.5}
  }
  MaterialExpressionMultiply_43 {
   class=MaterialExpressionMultiply
   inputs [
    A {from=MaterialExpressionAdd_37}
    B {from=""}
   ]
   properties {constB=0.5}
  }
  MaterialExpressionMultiply_44 {
   class=MaterialExpressionMultiply
   inputs [
    A {from=""}
    B {from=""}
   ]
  }
  MaterialExpressionMultiply_45 {
   class=MaterialExpressionMultiply
   inputs [
    A {from=MaterialExpressionNamedRerouteUsage_23}
    B {from=MaterialExpressionNamedRerouteUsage_34}
   ]
  }
  MaterialExpressionMultiply_46 {
   class=MaterialExpressionMultiply
   inputs [
    A {from=MaterialExpressionLinearInterpolate_14}
    B {from=MaterialExpressionDistance_2}
   ]
  }
  MaterialExpressionMultiply_47 {
   class=MaterialExpressionMultiply
   inputs [
    A {from=MaterialExpressionSaturate_9}
    B {from=MaterialExpressionScalarParameter_51}
   ]
  }
  MaterialExpressionMultiply_48 {
   class=MaterialExpressionMultiply
   inputs [
    A {from=MaterialExpressionAdd_38}
    B {from=""}
   ]
   properties {constB=0.5}
  }
  MaterialExpressionMultiply_49 {
   class=MaterialExpressionMultiply
   inputs [
    A {from=MaterialExpressionOneMinus_14}
    B {from=MaterialExpressionScalarParameter_58}
   ]
  }
  MaterialExpressionMultiply_5 {
   class=MaterialExpressionMultiply
   inputs [
    A {from=MaterialExpressionReroute_13}
    B {from=""}
   ]
   properties {constB=-1}
  }
  MaterialExpressionMultiply_50 {
   class=MaterialExpressionMultiply
   inputs [
    A {from=MaterialExpressionAdd_29}
    B {from=""}
   ]
   properties {constB=0.05}
  }
  MaterialExpressionMultiply_51 {
   class=MaterialExpressionMultiply
   inputs [
    A {from=MaterialExpressionAdd_39}
    B {from=MaterialExpressionScalarParameter_31}
   ]
  }
  MaterialExpressionMultiply_52 {
   class=MaterialExpressionMultiply
   inputs [
    A {from=MaterialExpressionSaturate_36}
    B {from=MaterialExpressionScalarParameter_32}
   ]
  }
  MaterialExpressionMultiply_53 {
   class=MaterialExpressionMultiply
   inputs [
    A {from=MaterialExpressionSine_0}
    B {from=""}
   ]
   properties {constB=0}
  }
  MaterialExpressionMultiply_54 {
   class=MaterialExpressionMultiply
   inputs [
    A {from=MaterialExpressionTime_3}
    B {from=""}
   ]
   properties {constB=0.5}
  }
  MaterialExpressionMultiply_55 {
   class=MaterialExpressionMultiply
   inputs [
    A {from=MaterialExpressionOneMinus_13}
    B {from=MaterialExpressionOneMinus_14}
   ]
  }
  MaterialExpressionMultiply_56 {
   class=MaterialExpressionMultiply
   inputs [
    A {from=MaterialExpressionPower_15}
    B {from=MaterialExpressionScalarParameter_37}
   ]
  }
  MaterialExpressionMultiply_57 {
   class=MaterialExpressionMultiply
   inputs [
    A {from=MaterialExpressionNamedRerouteUsage_3}
    B {from=MaterialExpressionConstant3Vector_14}
   ]
  }
  MaterialExpressionMultiply_58 {
   class=MaterialExpressionMultiply
   inputs [
    A {from=MaterialExpressionAdd_11}
    B {from=""}
   ]
   properties {constB=0.5}
  }
  MaterialExpressionMultiply_59 {
   class=MaterialExpressionMultiply
   inputs [
    A {from=MaterialExpressionLinearInterpolate_25}
    B {from=MaterialExpressionDivide_11}
   ]
  }
  MaterialExpressionMultiply_6 {
   class=MaterialExpressionMultiply
   inputs [
    A {from=MaterialExpressionNormalize_1}
    B {from=""}
   ]
   properties {constB=6}
  }
  MaterialExpressionMultiply_62 {
   class=MaterialExpressionMultiply
   inputs [
    A {from=MaterialExpressionReroute_32}
    B {from=""}
   ]
   properties {constB=0.02}
  }
  MaterialExpressionMultiply_63 {
   class=MaterialExpressionMultiply
   inputs [
    A {from=MaterialExpressionVertexNormalWS_5}
    B {from=MaterialExpressionMultiply_67}
   ]
  }
  MaterialExpressionMultiply_64 {
   class=MaterialExpressionMultiply
   inputs [
    A {from=MaterialExpressionSubtract_0}
    B {from=""}
   ]
   properties {constB=0.6}
  }
  MaterialExpressionMultiply_67 {
   class=MaterialExpressionMultiply
   inputs [
    A {from=MaterialExpressionAdd_10}
    B {from=""}
   ]
   properties {constB=0.5}
  }
  MaterialExpressionMultiply_69 {
   class=MaterialExpressionMultiply
   inputs [
    A {from=MaterialExpressionMaterialFunctionCall_1, output=Result}
    B {from=MaterialExpressionReroute_24}
   ]
  }
  MaterialExpressionMultiply_7 {
   class=MaterialExpressionMultiply
   inputs [
    A {from=MaterialExpressionMaterialFunctionCall_4, output=Result}
    B {from=MaterialExpressionReroute_22}
   ]
  }
  MaterialExpressionMultiply_71 {
   class=MaterialExpressionMultiply
   inputs [
    A {from=MaterialExpressionMaterialFunctionCall_17, output=Result}
    B {from=MaterialExpressionReroute_25}
   ]
  }
  MaterialExpressionMultiply_74 {
   class=MaterialExpressionMultiply
   inputs [
    A {from=MaterialExpressionDotProduct_8}
    B {from=""}
   ]
   properties {constB=2}
  }
  MaterialExpressionMultiply_75 {
   class=MaterialExpressionMultiply
   inputs [
    A {from=MaterialExpressionMultiply_74}
    B {from=MaterialExpressionReroute_15}
   ]
  }
  MaterialExpressionMultiply_76 {
   class=MaterialExpressionMultiply
   inputs [
    A {from=MaterialExpressionMaterialFunctionCall_41, output=Result}
    B {from=""}
   ]
  }
  MaterialExpressionMultiply_77 {
   class=MaterialExpressionMultiply
   inputs [
    A {from=MaterialExpressionPower_11}
    B {from=MaterialExpressionScalarParameter_2}
   ]
   properties {constB=0.1}
  }
  MaterialExpressionMultiply_79 {
   class=MaterialExpressionMultiply
   inputs [
    A {from=MaterialExpressionAdd_31}
    B {from=MaterialExpressionReroute_12}
   ]
  }
  MaterialExpressionMultiply_8 {
   class=MaterialExpressionMultiply
   inputs [
    A {from=MaterialExpressionScalarParameter_5}
    B {from=""}
   ]
   properties {constB=-0.55}
  }
  MaterialExpressionMultiply_80 {
   class=MaterialExpressionMultiply
   inputs [
    A {from=MaterialExpressionPower_12}
    B {from=MaterialExpressionScalarParameter_16}
   ]
   properties {constB=0.5}
  }
  MaterialExpressionMultiply_81 {
   class=MaterialExpressionMultiply
   inputs [
    A {from=MaterialExpressionDotProduct_9}
    B {from=""}
   ]
   properties {constB=-1}
  }
  MaterialExpressionMultiply_9 {
   class=MaterialExpressionMultiply
   inputs [
    A {from=MaterialExpressionAbs_7}
    B {from=""}
   ]
   properties {constB=-1}
  }
  MaterialExpressionMultiply_90 {
   class=MaterialExpressionMultiply
   inputs [
    A {from=MaterialExpressionOneMinus_10}
    B {from=""}
   ]
   properties {constB=0.0333}
  }
  MaterialExpressionMultiply_92 {
   class=MaterialExpressionMultiply
   inputs [
    A {from=MaterialExpressionMultiply_19}
    B {from=MaterialExpressionScalarParameter_55}
   ]
  }
  MaterialExpressionNamedRerouteDeclaration_0 {
   class=MaterialExpressionNamedRerouteDeclaration
   inputs [None {from=MaterialExpressionMaterialFunctionCall_7, output=Result}]
   properties {name=RebuiltDepth, nodeColor=<rgba 1 0 0.8203 1>}
  }
  MaterialExpressionNamedRerouteDeclaration_1 {
   class=MaterialExpressionNamedRerouteDeclaration
   inputs [None {from=MaterialExpressionDotProduct_5}]
   properties {name=Camera_Z, nodeColor=<rgba 0.8516 0 1 1>}
  }
  MaterialExpressionNamedRerouteDeclaration_10 {
   class=MaterialExpressionNamedRerouteDeclaration
   inputs [None {from=MaterialExpressionDotProduct_11}]
   properties {name=AxisDot, nodeColor=<rgba 0.2656 0 1 1>}
  }
  MaterialExpressionNamedRerouteDeclaration_11 {
   class=MaterialExpressionNamedRerouteDeclaration
   inputs [None {from=MaterialExpressionMultiply_33}]
   properties {name=AxialVector, nodeColor=<rgba 0.0547 1 0 1>}
  }
  MaterialExpressionNamedRerouteDeclaration_12 {
   class=MaterialExpressionNamedRerouteDeclaration
   inputs [None {from=MaterialExpressionSubtract_19}]
   properties {name=RadialVector, nodeColor=<rgba 1 0 0.375 1>}
  }
  MaterialExpressionNamedRerouteDeclaration_13 {
   class=MaterialExpressionNamedRerouteDeclaration
   inputs [None {from=MaterialExpressionLength_1}]
   properties {name=RadialDistance, nodeColor=<rgba 0 0.6953 1 1>}
  }
  MaterialExpressionNamedRerouteDeclaration_14 {
   class=MaterialExpressionNamedRerouteDeclaration
   inputs [None {from=MaterialExpressionOneMinus_1}]
   properties {name=RadiusMask, nodeColor=<rgba 1 0.9844 0 1>}
  }
  MaterialExpressionNamedRerouteDeclaration_15 {
   class=MaterialExpressionNamedRerouteDeclaration
   inputs [None {from=MaterialExpressionMultiply_34}]
   properties {name=LengthMask, nodeColor=<rgba 0.6641 0 1 1>}
  }
  MaterialExpressionNamedRerouteDeclaration_16 {
   class=MaterialExpressionNamedRerouteDeclaration
   inputs [None {from=MaterialExpressionAdd_39}]
   properties {name=ClyinderPosition, nodeColor=<rgba 0 0.0156 1 1>}
  }
  MaterialExpressionNamedRerouteDeclaration_17 {
   class=MaterialExpressionNamedRerouteDeclaration
   inputs [None {from=MaterialExpressionLinearInterpolate_20}]
   properties {name=CylinderMask, nodeColor=<rgba 1 0.0234 0 1>}
  }
  MaterialExpressionNamedRerouteDeclaration_18 {
   class=MaterialExpressionNamedRerouteDeclaration
   inputs [None {from=MaterialExpressionVectorParameter_5, output=RGB}]
   properties {name=CylinderRotation, nodeColor=<rgba 0.3359 1 0 1>}
  }
  MaterialExpressionNamedRerouteDeclaration_19 {
   class=MaterialExpressionNamedRerouteDeclaration
   inputs [None {from=MaterialExpressionClamp_0}]
   properties {name=DynamicZMin, nodeColor=<rgba 0 0.9531 1 1>}
  }
  MaterialExpressionNamedRerouteDeclaration_2 {
   class=MaterialExpressionNamedRerouteDeclaration
   inputs [None {from=MaterialExpressionScalarParameter_22}]
   properties {name=UseClothingMask, nodeColor=<rgba 0.736 0.1645 1 1>}
  }
  MaterialExpressionNamedRerouteDeclaration_20 {
   class=MaterialExpressionNamedRerouteDeclaration
   inputs [None {from=MaterialExpressionClamp_1}]
   properties {name=DynamicZMax, nodeColor=<rgba 0 0.9531 1 1>}
  }
  MaterialExpressionNamedRerouteDeclaration_21 {
   class=MaterialExpressionNamedRerouteDeclaration
   inputs [None {from=MaterialExpressionNormalize_7}]
   properties {name=ZAxis, nodeColor=<rgba 1 0.3516 0 1>}
  }
  MaterialExpressionNamedRerouteDeclaration_22 {
   class=MaterialExpressionNamedRerouteDeclaration
   inputs [None {from=MaterialExpressionNamedRerouteDeclaration_16}]
   properties {name=ClippingPos, nodeColor=<rgba 0.5938 0 1 1>}
  }
  MaterialExpressionNamedRerouteDeclaration_23 {
   class=MaterialExpressionNamedRerouteDeclaration
   inputs [None {from=MaterialExpressionScalarParameter_60}]
   properties {name=Clipping Plane Radius, nodeColor=<rgba 0 1 0.2734 1>}
  }
  MaterialExpressionNamedRerouteDeclaration_24 {
   class=MaterialExpressionNamedRerouteDeclaration
   inputs [None {from=MaterialExpressionLinearInterpolate_25}]
   properties {name=ClippingMask2, nodeColor=<rgba 1 0 0.0469 1>}
  }
  MaterialExpressionNamedRerouteDeclaration_25 {
   class=MaterialExpressionNamedRerouteDeclaration
   inputs [None {from=MaterialExpressionMultiply_56}]
   properties {name=DynamicTilt, nodeColor=<rgba 0 0.9531 1 1>}
  }
  MaterialExpressionNamedRerouteDeclaration_26 {
   class=MaterialExpressionNamedRerouteDeclaration
   inputs [None {from=MaterialExpressionStaticBoolParameter_0}]
   properties {name=DebugClippingMask, nodeColor=<rgba 0.7578 0 1 1>}
  }
  MaterialExpressionNamedRerouteDeclaration_3 {
   class=MaterialExpressionNamedRerouteDeclaration
   inputs [None {from=MaterialExpressionSaturate_25}]
   properties {name=BackHeadTansition, nodeColor=<rgba 0 1 0.8828 1>}
  }
  MaterialExpressionNamedRerouteDeclaration_4 {
   class=MaterialExpressionNamedRerouteDeclaration
   inputs [None {from=MaterialExpressionTransform_1}]
   properties {name=CameraDirection, nodeColor=<rgba 1 0 0.1875 1>}
  }
  MaterialExpressionNamedRerouteDeclaration_5 {
   class=MaterialExpressionNamedRerouteDeclaration
   inputs [None {from=MaterialExpressionSphereMask_0}]
   properties {name=ClothingMaskCylinder, nodeColor=<rgba 0.9219 0 1 1>}
  }
  MaterialExpressionNamedRerouteDeclaration_6 {
   class=MaterialExpressionNamedRerouteDeclaration
   inputs [None {from=MaterialExpressionSubtract_16}]
   properties {name=ToPixel, nodeColor=<rgba 1 0.5859 0 1>}
  }
  MaterialExpressionNamedRerouteDeclaration_7 {
   class=MaterialExpressionNamedRerouteDeclaration
   inputs [None {from=MaterialExpressionSaturate_28}]
   properties {name=EarPlane, nodeColor=<rgba 0.9219 1 0 1>}
  }
  MaterialExpressionNamedRerouteDeclaration_8 {
   class=MaterialExpressionNamedRerouteDeclaration
   inputs [None {from=MaterialExpressionSaturate_26}]
   properties {name=ClothingClippingPlane, nodeColor=<rgba 0 0.2031 1 1>}
  }
  MaterialExpressionNamedRerouteDeclaration_9 {
   class=MaterialExpressionNamedRerouteDeclaration
   inputs [None {from=MaterialExpressionVertexInterpolator_2, output=PS}]
   properties {name=LocalPos, nodeColor=<rgba 0 1 0.1094 1>}
  }
  MaterialExpressionNamedRerouteUsage_0 {class=MaterialExpressionNamedRerouteUsage}
  MaterialExpressionNamedRerouteUsage_1 {class=MaterialExpressionNamedRerouteUsage}
  MaterialExpressionNamedRerouteUsage_10 {class=MaterialExpressionNamedRerouteUsage}
  MaterialExpressionNamedRerouteUsage_11 {class=MaterialExpressionNamedRerouteUsage}
  MaterialExpressionNamedRerouteUsage_12 {class=MaterialExpressionNamedRerouteUsage}
  MaterialExpressionNamedRerouteUsage_13 {class=MaterialExpressionNamedRerouteUsage}
  MaterialExpressionNamedRerouteUsage_14 {class=MaterialExpressionNamedRerouteUsage}
  MaterialExpressionNamedRerouteUsage_15 {class=MaterialExpressionNamedRerouteUsage}
  MaterialExpressionNamedRerouteUsage_16 {class=MaterialExpressionNamedRerouteUsage}
  MaterialExpressionNamedRerouteUsage_17 {class=MaterialExpressionNamedRerouteUsage}
  MaterialExpressionNamedRerouteUsage_18 {class=MaterialExpressionNamedRerouteUsage}
  MaterialExpressionNamedRerouteUsage_19 {class=MaterialExpressionNamedRerouteUsage}
  MaterialExpressionNamedRerouteUsage_2 {class=MaterialExpressionNamedRerouteUsage}
  MaterialExpressionNamedRerouteUsage_20 {class=MaterialExpressionNamedRerouteUsage}
  MaterialExpressionNamedRerouteUsage_21 {class=MaterialExpressionNamedRerouteUsage}
  MaterialExpressionNamedRerouteUsage_22 {class=MaterialExpressionNamedRerouteUsage}
  MaterialExpressionNamedRerouteUsage_23 {class=MaterialExpressionNamedRerouteUsage}
  MaterialExpressionNamedRerouteUsage_24 {class=MaterialExpressionNamedRerouteUsage}
  MaterialExpressionNamedRerouteUsage_25 {class=MaterialExpressionNamedRerouteUsage}
  MaterialExpressionNamedRerouteUsage_26 {class=MaterialExpressionNamedRerouteUsage}
  MaterialExpressionNamedRerouteUsage_27 {class=MaterialExpressionNamedRerouteUsage}
  MaterialExpressionNamedRerouteUsage_28 {class=MaterialExpressionNamedRerouteUsage}
  MaterialExpressionNamedRerouteUsage_29 {class=MaterialExpressionNamedRerouteUsage}
  MaterialExpressionNamedRerouteUsage_3 {class=MaterialExpressionNamedRerouteUsage}
  MaterialExpressionNamedRerouteUsage_30 {class=MaterialExpressionNamedRerouteUsage}
  MaterialExpressionNamedRerouteUsage_31 {class=MaterialExpressionNamedRerouteUsage}
  MaterialExpressionNamedRerouteUsage_32 {class=MaterialExpressionNamedRerouteUsage}
  MaterialExpressionNamedRerouteUsage_33 {class=MaterialExpressionNamedRerouteUsage}
  MaterialExpressionNamedRerouteUsage_34 {class=MaterialExpressionNamedRerouteUsage}
  MaterialExpressionNamedRerouteUsage_4 {class=MaterialExpressionNamedRerouteUsage}
  MaterialExpressionNamedRerouteUsage_5 {class=MaterialExpressionNamedRerouteUsage}
  MaterialExpressionNamedRerouteUsage_6 {class=MaterialExpressionNamedRerouteUsage}
  MaterialExpressionNamedRerouteUsage_7 {class=MaterialExpressionNamedRerouteUsage}
  MaterialExpressionNamedRerouteUsage_8 {class=MaterialExpressionNamedRerouteUsage}
  MaterialExpressionNamedRerouteUsage_9 {class=MaterialExpressionNamedRerouteUsage}
  MaterialExpressionNormalize_0 {class=MaterialExpressionNormalize, inputs [VectorInput {from=MaterialExpressionSubtract_1, output=""}]}
  MaterialExpressionNormalize_1 {class=MaterialExpressionNormalize, inputs [VectorInput {from=MaterialExpressionSubtract_12, output=""}]}
  MaterialExpressionNormalize_2 {class=MaterialExpressionNormalize, inputs [VectorInput {from=MaterialExpressionAdd_3, output=""}]}
  MaterialExpressionNormalize_3 {class=MaterialExpressionNormalize, inputs [VectorInput {from=MaterialExpressionSubtract_20, output=""}]}
  MaterialExpressionNormalize_4 {class=MaterialExpressionNormalize, inputs [VectorInput {from=MaterialExpressionSubtract_15, output=""}]}
  MaterialExpressionNormalize_5 {class=MaterialExpressionNormalize, inputs [VectorInput {from=MaterialExpressionNamedRerouteDeclaration_18, output=""}]}
  MaterialExpressionNormalize_6 {class=MaterialExpressionNormalize, inputs [VectorInput {from=MaterialExpressionAdd_47, output=""}]}
  MaterialExpressionNormalize_7 {class=MaterialExpressionNormalize, inputs [VectorInput {from=MaterialExpressionNamedRerouteDeclaration_18, output=""}]}
  MaterialExpressionNormalize_8 {class=MaterialExpressionNormalize, inputs [VectorInput {from=MaterialExpressionLinearInterpolate_9, output=""}]}
  MaterialExpressionNormalize_9 {class=MaterialExpressionNormalize, inputs [VectorInput {from=MaterialExpressionNamedRerouteDeclaration_21, output=""}]}
  MaterialExpressionObjectPositionWS_13 {class=MaterialExpressionObjectPositionWS, inputs [], properties {}}
  MaterialExpressionOneMinus_0 {class=MaterialExpressionOneMinus, inputs [None {from=MaterialExpressionLinearInterpolate_22}]}
  MaterialExpressionOneMinus_1 {class=MaterialExpressionOneMinus, inputs [None {from=MaterialExpressionSaturate_13}]}
  MaterialExpressionOneMinus_10 {class=MaterialExpressionOneMinus, inputs [None {from=MaterialExpressionComponentMask_3}]}
  MaterialExpressionOneMinus_11 {class=MaterialExpressionOneMinus, inputs [None {from=MaterialExpressionSaturate_31}]}
  MaterialExpressionOneMinus_12 {class=MaterialExpressionOneMinus, inputs [None {from=MaterialExpressionMultiply_41}]}
  MaterialExpressionOneMinus_13 {class=MaterialExpressionOneMinus, inputs [None {from=MaterialExpressionMultiply_42}]}
  MaterialExpressionOneMinus_14 {class=MaterialExpressionOneMinus, inputs [None {from=MaterialExpressionMultiply_48}]}
  MaterialExpressionOneMinus_17 {class=MaterialExpressionOneMinus, inputs [None {from=MaterialExpressionMaterialFunctionCall_24, output=Opacity Mask}]}
  MaterialExpressionOneMinus_18 {class=MaterialExpressionOneMinus, inputs [None {from=MaterialExpressionLinearInterpolate_25}]}
  MaterialExpressionOneMinus_2 {class=MaterialExpressionOneMinus, inputs [None {from=MaterialExpressionSaturate_1}]}
  MaterialExpressionOneMinus_3 {class=MaterialExpressionOneMinus, inputs [None {from=MaterialExpressionPower_4}]}
  MaterialExpressionOneMinus_4 {class=MaterialExpressionOneMinus, inputs [None {from=MaterialExpressionMultiply_38}]}
  MaterialExpressionOneMinus_5 {class=MaterialExpressionOneMinus, inputs [None {from=MaterialExpressionSaturate_27}]}
  MaterialExpressionOneMinus_6 {class=MaterialExpressionOneMinus, inputs [None {from=MaterialExpressionMultiply_43}]}
  MaterialExpressionOneMinus_7 {class=MaterialExpressionOneMinus, inputs [None {from=MaterialExpressionMultiply_43}]}
  MaterialExpressionOneMinus_8 {class=MaterialExpressionOneMinus, inputs [None {from=MaterialExpressionMaterialFunctionCall_46, output=Result}]}
  MaterialExpressionOneMinus_9 {class=MaterialExpressionOneMinus, inputs [None {from=MaterialExpressionMaterialFunctionCall_29, output=Result}]}
  MaterialExpressionPanner_0 {
   class=MaterialExpressionPanner
   inputs [
    Coordinate {from=MaterialExpressionComponentMask_24}
    Time {from=MaterialExpressionTime_4}
    Speed {from=MaterialExpressionReroute_29}
   ]
  }
  MaterialExpressionPanner_2 {
   class=MaterialExpressionPanner
   inputs [
    Coordinate {from=MaterialExpressionComponentMask_23}
    Time {from=MaterialExpressionTime_4}
    Speed {from=MaterialExpressionReroute_29}
   ]
  }
  MaterialExpressionPanner_9 {
   class=MaterialExpressionPanner
   inputs [
    Coordinate {from=MaterialExpressionComponentMask_0}
    Time {from=MaterialExpressionTime_4}
    Speed {from=MaterialExpressionReroute_29}
   ]
  }
  MaterialExpressionPower_0 {
   class=MaterialExpressionPower
   inputs [
    Base {from=MaterialExpressionMultiply_35}
    Exp {from=MaterialExpressionScalarParameter_14}
   ]
  }
  MaterialExpressionPower_1 {
   class=MaterialExpressionPower
   inputs [
    Base {from=MaterialExpressionMax_0}
    Exp {from=MaterialExpressionConstant_0}
   ]
  }
  MaterialExpressionPower_10 {
   class=MaterialExpressionPower
   inputs [
    Base {from=MaterialExpressionSaturate_35}
    Exp {from=MaterialExpressionScalarParameter_48}
   ]
  }
  MaterialExpressionPower_11 {
   class=MaterialExpressionPower
   inputs [
    Base {from=MaterialExpressionMax_1}
    Exp {from=MaterialExpressionConstant_13}
   ]
  }
  MaterialExpressionPower_12 {
   class=MaterialExpressionPower
   inputs [
    Base {from=MaterialExpressionMultiply_81}
    Exp {from=MaterialExpressionConstant_14}
   ]
  }
  MaterialExpressionPower_13 {
   class=MaterialExpressionPower
   inputs [
    Base {from=MaterialExpressionSaturate_8}
    Exp {from=MaterialExpressionScalarParameter_54}
   ]
  }
  MaterialExpressionPower_14 {
   class=MaterialExpressionPower
   inputs [
    Base {from=MaterialExpressionMultiply_58}
    Exp {from=MaterialExpressionScalarParameter_59}
   ]
  }
  MaterialExpressionPower_15 {
   class=MaterialExpressionPower
   inputs [
    Base {from=MaterialExpressionSaturate_32}
    Exp {from=MaterialExpressionScalarParameter_38}
   ]
  }
  MaterialExpressionPower_16 {
   class=MaterialExpressionPower
   inputs [
    Base {from=MaterialExpressionSaturate_38}
    Exp {from=MaterialExpressionScalarParameter_63}
   ]
  }
  MaterialExpressionPower_2 {
   class=MaterialExpressionPower
   inputs [
    Base {from=MaterialExpressionMax_3}
    Exp {from=MaterialExpressionConstant_10}
   ]
  }
  MaterialExpressionPower_3 {
   class=MaterialExpressionPower
   inputs [
    Base {from=MaterialExpressionOneMinus_4}
    Exp {from=MaterialExpressionScalarParameter_23}
   ]
  }
  MaterialExpressionPower_4 {
   class=MaterialExpressionPower
   inputs [
    Base {from=MaterialExpressionAbs_1}
    Exp {from=""}
   ]
  }
  MaterialExpressionPower_5 {
   class=MaterialExpressionPower
   inputs [
    Base {from=MaterialExpressionSaturate_36}
    Exp {from=MaterialExpressionScalarParameter_32}
   ]
  }
  MaterialExpressionPower_6 {
   class=MaterialExpressionPower
   inputs [
    Base {from=MaterialExpressionSaturate_5}
    Exp {from=MaterialExpressionScalarParameter_9}
   ]
  }
  MaterialExpressionPower_7 {
   class=MaterialExpressionPower
   inputs [
    Base {from=MaterialExpressionSaturate_30}
    Exp {from=MaterialExpressionScalarParameter_40}
   ]
  }
  MaterialExpressionPower_8 {
   class=MaterialExpressionPower
   inputs [
    Base {from=MaterialExpressionSaturate_35}
    Exp {from=MaterialExpressionScalarParameter_47}
   ]
  }
  MaterialExpressionPower_9 {
   class=MaterialExpressionPower
   inputs [
    Base {from=MaterialExpressionMultiply_43}
    Exp {from=MaterialExpressionScalarParameter_44}
   ]
  }
  MaterialExpressionPreSkinnedNormal_2 {class=MaterialExpressionPreSkinnedNormal, inputs [], properties {}}
  MaterialExpressionPreSkinnedPosition_1 {class=MaterialExpressionPreSkinnedPosition, inputs [], properties {}}
  MaterialExpressionQualitySwitch_0 {
   class=MaterialExpressionQualitySwitch
   inputs [
    Default {from=MaterialExpressionTransform_4, output=""}
    Low {from=MaterialExpressionTransform_0, output=""}
    High {from="", output=""}
    Medium {from="", output=""}
    Epic {from="", output=""}
   ]
   properties {}
  }
  MaterialExpressionQualitySwitch_1 {
   class=MaterialExpressionQualitySwitch
   inputs [
    Default {from=MaterialExpressionAdd_9, output=""}
    Low {from=MaterialExpressionReroute_7, output=""}
    High {from="", output=""}
    Medium {from="", output=""}
    Epic {from="", output=""}
   ]
   properties {}
  }
  MaterialExpressionReroute_0 {class=MaterialExpressionReroute, inputs [None {from=MaterialExpressionConstant3Vector_6}]}
  MaterialExpressionReroute_1 {class=MaterialExpressionReroute, inputs [None {from=MaterialExpressionMultiply_39}]}
  MaterialExpressionReroute_10 {class=MaterialExpressionReroute, inputs [None {from=MaterialExpressionPower_13}]}
  MaterialExpressionReroute_11 {class=MaterialExpressionReroute, inputs [None {from=MaterialExpressionSaturate_4}]}
  MaterialExpressionReroute_12 {class=MaterialExpressionReroute, inputs [None {from=MaterialExpressionSaturate_9}]}
  MaterialExpressionReroute_13 {class=MaterialExpressionReroute, inputs [None {from=MaterialExpressionDotProduct_1}]}
  MaterialExpressionReroute_14 {class=MaterialExpressionReroute, inputs [None {from=MaterialExpressionMultiply_92}]}
  MaterialExpressionReroute_15 {class=MaterialExpressionReroute, inputs [None {from=MaterialExpressionTransform_2}]}
  MaterialExpressionReroute_17 {class=MaterialExpressionReroute, inputs [None {from=MaterialExpressionReroute_10}]}
  MaterialExpressionReroute_18 {class=MaterialExpressionReroute, inputs [None {from=MaterialExpressionScalarParameter_0}]}
  MaterialExpressionReroute_19 {class=MaterialExpressionReroute, inputs [None {from=MaterialExpressionScalarParameter_1}]}
  MaterialExpressionReroute_20 {class=MaterialExpressionReroute, inputs [None {from=MaterialExpressionConstant3Vector_0}]}
  MaterialExpressionReroute_21 {class=MaterialExpressionReroute, inputs [None {from=MaterialExpressionDivide_0}]}
  MaterialExpressionReroute_22 {class=MaterialExpressionReroute, inputs [None {from=MaterialExpressionComponentMask_15}]}
  MaterialExpressionReroute_23 {class=MaterialExpressionReroute, inputs [None {from=MaterialExpressionVertexInterpolator_5, output=PS}]}
  MaterialExpressionReroute_24 {class=MaterialExpressionReroute, inputs [None {from=MaterialExpressionComponentMask_14}]}
  MaterialExpressionReroute_25 {class=MaterialExpressionReroute, inputs [None {from=MaterialExpressionComponentMask_2}]}
  MaterialExpressionReroute_26 {class=MaterialExpressionReroute, inputs [None {from=MaterialExpressionTextureObject_4}]}
  MaterialExpressionReroute_27 {class=MaterialExpressionReroute, inputs [None {from=MaterialExpressionAdd_21}]}
  MaterialExpressionReroute_28 {class=MaterialExpressionReroute, inputs [None {from=MaterialExpressionNormalize_2}]}
  MaterialExpressionReroute_29 {class=MaterialExpressionReroute, inputs [None {from=MaterialExpressionAppendVector_10}]}
  MaterialExpressionReroute_3 {class=MaterialExpressionReroute, inputs [None {from=MaterialExpressionDivide_1}]}
  MaterialExpressionReroute_30 {class=MaterialExpressionReroute, inputs [None {from=MaterialExpressionVertexInterpolator_5, output=PS}]}
  MaterialExpressionReroute_31 {class=MaterialExpressionReroute, inputs [None {from=MaterialExpressionOneMinus_3}]}
  MaterialExpressionReroute_32 {class=MaterialExpressionReroute, inputs [None {from=MaterialExpressionPreSkinnedPosition_1}]}
  MaterialExpressionReroute_34 {class=MaterialExpressionReroute, inputs [None {from=MaterialExpressionSubtract_4}]}
  MaterialExpressionReroute_4 {class=MaterialExpressionReroute, inputs [None {from=MaterialExpressionSaturate_14}]}
  MaterialExpressionReroute_5 {class=MaterialExpressionReroute, inputs [None {from=MaterialExpressionReroute_9}]}
  MaterialExpressionReroute_6 {class=MaterialExpressionReroute, inputs [None {from=MaterialExpressionLinearInterpolate_8}]}
  MaterialExpressionReroute_7 {class=MaterialExpressionReroute, inputs [None {from=MaterialExpressionLinearInterpolate_0}]}
  MaterialExpressionReroute_8 {class=MaterialExpressionReroute, inputs [None {from=MaterialExpressionMultiply_25}]}
  MaterialExpressionReroute_9 {class=MaterialExpressionReroute, inputs [None {from=MaterialExpressionSaturate_12}]}
  MaterialExpressionRotateAboutAxis_0 {
   class=MaterialExpressionRotateAboutAxis
   inputs [
    NormalizedRotationAxis {from=MaterialExpressionConstant3Vector_11, output=""}
    RotationAngle {from=MaterialExpressionScalarParameter_21, output=""}
    PivotPoint {from=MaterialExpressionConvert_1, output=""}
    Position {from=MaterialExpressionNamedRerouteUsage_11, output=""}
   ]
   properties {}
  }
  MaterialExpressionSaturate_0 {class=MaterialExpressionSaturate, inputs [None {from=MaterialExpressionAdd_5}]}
  MaterialExpressionSaturate_1 {class=MaterialExpressionSaturate, inputs [None {from=MaterialExpressionComponentMask_7}]}
  MaterialExpressionSaturate_10 {class=MaterialExpressionSaturate, inputs [None {from=MaterialExpressionMultiply_30}]}
  MaterialExpressionSaturate_11 {class=MaterialExpressionSaturate, inputs [None {from=MaterialExpressionMaterialFunctionCall_30, output=Mask_Top}]}
  MaterialExpressionSaturate_12 {class=MaterialExpressionSaturate, inputs [None {from=MaterialExpressionMultiply_4}]}
  MaterialExpressionSaturate_13 {class=MaterialExpressionSaturate, inputs [None {from=MaterialExpressionDivide_5}]}
  MaterialExpressionSaturate_14 {class=MaterialExpressionSaturate, inputs [None {from=MaterialExpressionMultiply_14}]}
  MaterialExpressionSaturate_15 {class=MaterialExpressionSaturate, inputs [None {from=MaterialExpressionAdd_4}]}
  MaterialExpressionSaturate_16 {class=MaterialExpressionSaturate, inputs [None {from=MaterialExpressionSubtract_6}]}
  MaterialExpressionSaturate_17 {class=MaterialExpressionSaturate, inputs [None {from=MaterialExpressionMultiply_77}]}
  MaterialExpressionSaturate_18 {class=MaterialExpressionSaturate, inputs [None {from=MaterialExpressionMultiply_80}]}
  MaterialExpressionSaturate_19 {class=MaterialExpressionSaturate, inputs [None {from=MaterialExpressionAdd_35}]}
  MaterialExpressionSaturate_2 {class=MaterialExpressionSaturate, inputs [None {from=MaterialExpressionAdd_13}]}
  MaterialExpressionSaturate_20 {class=MaterialExpressionSaturate, inputs [None {from=MaterialExpressionMultiply_23}]}
  MaterialExpressionSaturate_21 {class=MaterialExpressionSaturate, inputs [None {from=MaterialExpressionMultiply_29}]}
  MaterialExpressionSaturate_22 {class=MaterialExpressionSaturate, inputs [None {from=MaterialExpressionMultiply_26}]}
  MaterialExpressionSaturate_23 {class=MaterialExpressionSaturate, inputs [None {from=MaterialExpressionAdd_26}]}
  MaterialExpressionSaturate_24 {class=MaterialExpressionSaturate, inputs [None {from=MaterialExpressionAdd_27}]}
  MaterialExpressionSaturate_25 {class=MaterialExpressionSaturate, inputs [None {from=MaterialExpressionAdd_32}]}
  MaterialExpressionSaturate_26 {class=MaterialExpressionSaturate, inputs [None {from=MaterialExpressionDivide_2}]}
  MaterialExpressionSaturate_27 {class=MaterialExpressionSaturate, inputs [None {from=MaterialExpressionDivide_4}]}
  MaterialExpressionSaturate_28 {class=MaterialExpressionSaturate, inputs [None {from=MaterialExpressionDivide_3}]}
  MaterialExpressionSaturate_29 {class=MaterialExpressionSaturate, inputs [None {from=MaterialExpressionAdd_41}]}
  MaterialExpressionSaturate_3 {class=MaterialExpressionSaturate, inputs [None {from=MaterialExpressionIf_0}]}
  MaterialExpressionSaturate_30 {class=MaterialExpressionSaturate, inputs [None {from=MaterialExpressionDivide_7}]}
  MaterialExpressionSaturate_31 {class=MaterialExpressionSaturate, inputs [None {from=MaterialExpressionDivide_8}]}
  MaterialExpressionSaturate_32 {class=MaterialExpressionSaturate, inputs [None {from=MaterialExpressionMultiply_55}]}
  MaterialExpressionSaturate_34 {class=MaterialExpressionSaturate, inputs [None {from=MaterialExpressionDivide_6}]}
  MaterialExpressionSaturate_35 {class=MaterialExpressionSaturate, inputs [None {from=MaterialExpressionAdd_40}]}
  MaterialExpressionSaturate_36 {class=MaterialExpressionSaturate, inputs [None {from=MaterialExpressionComponentMask_11}]}
  MaterialExpressionSaturate_37 {class=MaterialExpressionSaturate, inputs [None {from=MaterialExpressionComponentMask_10}]}
  MaterialExpressionSaturate_38 {class=MaterialExpressionSaturate, inputs [None {from=MaterialExpressionDivide_11}]}
  MaterialExpressionSaturate_4 {class=MaterialExpressionSaturate, inputs [None {from=MaterialExpressionSubtract_2}]}
  MaterialExpressionSaturate_5 {class=MaterialExpressionSaturate, inputs [None {from=MaterialExpressionAdd_15}]}
  MaterialExpressionSaturate_6 {class=MaterialExpressionSaturate, inputs [None {from=MaterialExpressionMultiply_13}]}
  MaterialExpressionSaturate_7 {class=MaterialExpressionSaturate, inputs [None {from=MaterialExpressionMaterialFunctionCall_10, output=UVChannel1}]}
  MaterialExpressionSaturate_8 {class=MaterialExpressionSaturate, inputs [None {from=MaterialExpressionOneMinus_8}]}
  MaterialExpressionSaturate_9 {class=MaterialExpressionSaturate, inputs [None {from=MaterialExpressionSubtract_7}]}
  MaterialExpressionScalarParameter_0 {class=MaterialExpressionScalarParameter, properties {parameterName=Roughness}}
  MaterialExpressionScalarParameter_1 {class=MaterialExpressionScalarParameter, properties {parameterName=Specular}}
  MaterialExpressionScalarParameter_10 {class=MaterialExpressionScalarParameter, properties {defaultValue=1.25, parameterName=BackHeadTransitionAdd}}
  MaterialExpressionScalarParameter_11 {class=MaterialExpressionScalarParameter, properties {parameterName=BackHeadTransitionBias}}
  MaterialExpressionScalarParameter_12 {class=MaterialExpressionScalarParameter, properties {defaultValue=10, parameterName=CylinderRadius}}
  MaterialExpressionScalarParameter_13 {class=MaterialExpressionScalarParameter, properties {defaultValue=50, parameterName=ClothingMaskCylinder_Radius}}
  MaterialExpressionScalarParameter_14 {class=MaterialExpressionScalarParameter, properties {defaultValue=1, parameterName=ClothingMask_CurvePower_X}}
  MaterialExpressionScalarParameter_15 {class=MaterialExpressionScalarParameter, properties {defaultValue=1, parameterName=ClothingMaskCylinder_Softness}}
  MaterialExpressionScalarParameter_16 {class=MaterialExpressionScalarParameter, properties {defaultValue=0.5, parameterName=Bubble_HighlightSpecular_Scale}}
  MaterialExpressionScalarParameter_17 {class=MaterialExpressionScalarParameter, properties {defaultValue=10, parameterName=ClothingClippingPlane_HeightMax}}
  MaterialExpressionScalarParameter_18 {class=MaterialExpressionScalarParameter, properties {defaultValue=1, parameterName=ClothingClippingPlane_HeightMin}}
  MaterialExpressionScalarParameter_19 {class=MaterialExpressionScalarParameter, properties {defaultValue=-13, parameterName=DepthBias_Top}}
  MaterialExpressionScalarParameter_2 {class=MaterialExpressionScalarParameter, properties {defaultValue=0.1, parameterName=Bubble_BackSpecular_Scale}}
  MaterialExpressionScalarParameter_20 {class=MaterialExpressionScalarParameter, properties {defaultValue=1, parameterName=DepthMask_Softness}}
  MaterialExpressionScalarParameter_21 {class=MaterialExpressionScalarParameter, properties {defaultValue=0.5, parameterName=DepthMask_Tilt, sliderMax=1}}
  MaterialExpressionScalarParameter_22 {
   class=MaterialExpressionScalarParameter
   properties {controlType=Enumeration, enumeration=/Game/Global/Materials/Enums/E_Mat_YesNo, parameterName=UseClothingMask}
  }
  MaterialExpressionScalarParameter_23 {class=MaterialExpressionScalarParameter, properties {defaultValue=256, parameterName=ClothingMask_CurvePower_Y}}
  MaterialExpressionScalarParameter_24 {class=MaterialExpressionScalarParameter, properties {parameterName=ClothingMask_Y_Offset}}
  MaterialExpressionScalarParameter_25 {class=MaterialExpressionScalarParameter, properties {defaultValue=-100, parameterName="CylinderZMin ( - )"}}
  MaterialExpressionScalarParameter_26 {class=MaterialExpressionScalarParameter, properties {defaultValue=1, parameterName=EarPlane_Softness}}
  MaterialExpressionScalarParameter_27 {class=MaterialExpressionScalarParameter, properties {defaultValue=1, parameterName=EarPlane_Offset}}
  MaterialExpressionScalarParameter_28 {class=MaterialExpressionScalarParameter, properties {parameterName=EarTransitionBias}}
  MaterialExpressionScalarParameter_29 {class=MaterialExpressionScalarParameter, properties {defaultValue=1, parameterName=CylinderSoftness_Top}}
  MaterialExpressionScalarParameter_3 {class=MaterialExpressionScalarParameter, properties {defaultValue=0.2, parameterName=HeadBubble_SpecularScale}}
  MaterialExpressionScalarParameter_30 {class=MaterialExpressionScalarParameter, properties {defaultValue=100, parameterName="CylinderZMax ( + )"}}
  MaterialExpressionScalarParameter_31 {class=MaterialExpressionScalarParameter, properties {defaultValue=1, parameterName=Pos_along_Z}}
  MaterialExpressionScalarParameter_32 {class=MaterialExpressionScalarParameter, properties {defaultValue=1, parameterName=ClothingMask_CurvePower_X}}
  MaterialExpressionScalarParameter_33 {class=MaterialExpressionScalarParameter, properties {defaultValue=10, parameterName=Disc Radius}}
  MaterialExpressionScalarParameter_34 {class=MaterialExpressionScalarParameter, properties {defaultValue=0.025, parameterName=Bubble Speed Y_1}}
  MaterialExpressionScalarParameter_35 {class=MaterialExpressionScalarParameter, properties {parameterName=CylinderZMax_BackHead_Offset}}
  MaterialExpressionScalarParameter_36 {class=MaterialExpressionScalarParameter, properties {defaultValue=-13, parameterName=DepthBias}}
  MaterialExpressionScalarParameter_37 {class=MaterialExpressionScalarParameter, properties {defaultValue=0.5, parameterName=Cylinder_TiltMax}}
  MaterialExpressionScalarParameter_38 {class=MaterialExpressionScalarParameter, properties {defaultValue=2, parameterName=Cylinder_CurvedPowerTilt}}
  MaterialExpressionScalarParameter_39 {
   class=MaterialExpressionScalarParameter
   properties {controlType=Enumeration, enumeration=&E_Mat_YesNo, parameterName="Cylinder_OnlyUseAxis_Z (Shows Length)"}
  }
  MaterialExpressionScalarParameter_4 {class=MaterialExpressionScalarParameter, properties {group=S22ChromeCharacters, parameterName=S22ChromeCharacters_strength, sliderMax=1}}
  MaterialExpressionScalarParameter_40 {class=MaterialExpressionScalarParameter, properties {defaultValue=50, parameterName=Disc FallOff}}
  MaterialExpressionScalarParameter_41 {class=MaterialExpressionScalarParameter, properties {defaultValue=1, parameterName=CylinderSoftness_Bottom}}
  MaterialExpressionScalarParameter_42 {class=MaterialExpressionScalarParameter, properties {defaultValue=11, parameterName=Radius_Max}}
  MaterialExpressionScalarParameter_43 {class=MaterialExpressionScalarParameter, properties {defaultValue=10, parameterName=Radius_Min}}
  MaterialExpressionScalarParameter_44 {class=MaterialExpressionScalarParameter, properties {defaultValue=1, parameterName=Radius_CurvePower}}
  MaterialExpressionScalarParameter_45 {class=MaterialExpressionScalarParameter, properties {defaultValue=1, parameterName=CylinderSoftness_Radius}}
  MaterialExpressionScalarParameter_46 {
   class=MaterialExpressionScalarParameter
   properties {controlType=Enumeration, enumeration=&E_Mat_YesNo, parameterName=Invert_CylinderMask}
  }
  MaterialExpressionScalarParameter_47 {class=MaterialExpressionScalarParameter, properties {defaultValue=10, parameterName=Cylinder_CurvedPoweZMin}}
  MaterialExpressionScalarParameter_48 {class=MaterialExpressionScalarParameter, properties {defaultValue=10, parameterName=Cylinder_CurvedPowerZMax}}
  MaterialExpressionScalarParameter_49 {class=MaterialExpressionScalarParameter, properties {defaultValue=25, parameterName=BubbleSize_A}}
  MaterialExpressionScalarParameter_5 {class=MaterialExpressionScalarParameter, properties {defaultValue=0.5, parameterName=Blend Sharpness, sliderMax=1}}
  MaterialExpressionScalarParameter_50 {class=MaterialExpressionScalarParameter, properties {defaultValue=10, parameterName=ZMin_Max}}
  MaterialExpressionScalarParameter_51 {class=MaterialExpressionScalarParameter, properties {defaultValue=0.25, parameterName=SHBubbleScale}}
  MaterialExpressionScalarParameter_52 {class=MaterialExpressionScalarParameter, properties {parameterName=SHContrast_LowValue}}
  MaterialExpressionScalarParameter_53 {class=MaterialExpressionScalarParameter, properties {defaultValue=1, parameterName=SHContrast_HighValue}}
  MaterialExpressionScalarParameter_54 {class=MaterialExpressionScalarParameter, properties {defaultValue=1, parameterName=SHPow}}
  MaterialExpressionScalarParameter_55 {class=MaterialExpressionScalarParameter, properties {defaultValue=0.33, parameterName=WPO_FinalScale}}
  MaterialExpressionScalarParameter_56 {class=MaterialExpressionScalarParameter, properties {defaultValue=1, parameterName=Debug Clip Mask View}}
  MaterialExpressionScalarParameter_57 {class=MaterialExpressionScalarParameter, properties {defaultValue=10, parameterName=ZMax_Max}}
  MaterialExpressionScalarParameter_58 {class=MaterialExpressionScalarParameter, properties {parameterName=CylinderBackView_Influence}}
  MaterialExpressionScalarParameter_59 {class=MaterialExpressionScalarParameter, properties {defaultValue=1, parameterName=DepthBiasTop_CurvePower}}
  MaterialExpressionScalarParameter_6 {class=MaterialExpressionScalarParameter, properties {defaultValue=1, parameterName=Eyes_FresnelContrast_bright}}
  MaterialExpressionScalarParameter_60 {class=MaterialExpressionScalarParameter, properties {defaultValue=100, parameterName=Clipping Plane Radius}}
  MaterialExpressionScalarParameter_61 {class=MaterialExpressionScalarParameter, properties {defaultValue=41.6444, parameterName=Clipping Mask FallOff}}
  MaterialExpressionScalarParameter_62 {class=MaterialExpressionScalarParameter, properties {defaultValue=1, parameterName=DepthMask_Height}}
  MaterialExpressionScalarParameter_63 {class=MaterialExpressionScalarParameter, properties {defaultValue=1, parameterName=Clipping Mask FallOff Curve}}
  MaterialExpressionScalarParameter_7 {class=MaterialExpressionScalarParameter, properties {defaultValue=-0.5, parameterName=Eyes_FresnelContrast_dark}}
  MaterialExpressionScalarParameter_8 {class=MaterialExpressionScalarParameter, properties {parameterName=Slurp_WPO_random_seed}}
  MaterialExpressionScalarParameter_9 {class=MaterialExpressionScalarParameter, properties {defaultValue=2, parameterName=BackHeadTransitionExp}}
  MaterialExpressionSceneDepth_0 {class=MaterialExpressionSceneDepth, inputs [UVs {from="", output=""}], properties {}}
  MaterialExpressionSetMaterialAttributes_0 {
   class=MaterialExpressionSetMaterialAttributes
   inputs [
    MaterialAttributes {from="", output=""}
    Base Color {from=MaterialExpressionReroute_20, output=""}
    Metallic {from=MaterialExpressionScalarParameter_1, output=""}
    Specular {from=MaterialExpressionReroute_19, output=""}
    Roughness {from=MaterialExpressionReroute_18, output=""}
    Emissive Color {from=MaterialExpressionStaticSwitch_1, output=""}
    Opacity {from=MaterialExpressionStaticSwitch_0, output=""}
    Normal {from="", output=""}
    World Position Offset {from=MaterialExpressionReroute_6, output=""}
   ]
   properties {attributeSetTypes [
     #1
     #2
     #3
     #4
     #5
     #6
     #7
     #8
    ]}
  }
  MaterialExpressionShadingPathSwitch_0 {
   class=MaterialExpressionShadingPathSwitch
   inputs [
    Default {from=MaterialExpressionQualitySwitch_0, output=""}
    Deferred {from=MaterialExpressionQualitySwitch_0, output=""}
    Forward {from=MaterialExpressionTransform_0, output=""}
    Mobile {from=MaterialExpressionTransform_0, output=""}
   ]
   properties {}
  }
  MaterialExpressionShadingPathSwitch_1 {
   class=MaterialExpressionShadingPathSwitch
   inputs [
    Default {from=MaterialExpressionQualitySwitch_1, output=""}
    Deferred {from=MaterialExpressionQualitySwitch_1, output=""}
    Forward {from=MaterialExpressionReroute_7, output=""}
    Mobile {from=MaterialExpressionReroute_7, output=""}
   ]
   properties {}
  }
  MaterialExpressionSine_0 {class=MaterialExpressionSine, inputs [None {from=MaterialExpressionFrac_1}]}
  MaterialExpressionSine_2 {class=MaterialExpressionSine, inputs [None {from=MaterialExpressionSubtract_18}]}
  MaterialExpressionSphereMask_0 {
   class=MaterialExpressionSphereMask
   inputs [
    A {from=MaterialExpressionNamedRerouteUsage_2, output=""}
    B {from=MaterialExpressionVectorParameter_4, output=RGB}
    Radius {from=MaterialExpressionScalarParameter_13, output=""}
    Hardness {from=MaterialExpressionScalarParameter_15, output=""}
   ]
   properties {}
  }
  MaterialExpressionStaticBoolParameter_0 {class=MaterialExpressionStaticBoolParameter, inputs [], properties {parameterName=DebugClippingMask}}
  MaterialExpressionStaticBool_0 {class=MaterialExpressionStaticBool, properties {value=true}}
  MaterialExpressionStaticBool_1 {class=MaterialExpressionStaticBool, properties {value=true}}
  MaterialExpressionStaticBool_2 {class=MaterialExpressionStaticBool, properties {value=true}}
  MaterialExpressionStaticBool_3 {class=MaterialExpressionStaticBool, properties {value=true}}
  MaterialExpressionStaticSwitchParameter_1 {
   class=MaterialExpressionStaticSwitchParameter
   inputs [
    True {from=MaterialExpressionAdd_28, output=""}
    False {from=MaterialExpressionAppendVector_5, output=""}
   ]
   properties {parameterName=UseCustomSH Fit}
  }
  MaterialExpressionStaticSwitch_0 {
   class=MaterialExpressionStaticSwitch
   inputs [
    True {from=MaterialExpressionConstant_3, output=""}
    False {from=MaterialExpressionReroute_1, output=""}
    Value {from=MaterialExpressionNamedRerouteUsage_18, output=""}
   ]
   properties {}
  }
  MaterialExpressionStaticSwitch_1 {
   class=MaterialExpressionStaticSwitch
   inputs [
    True {from=MaterialExpressionNamedRerouteUsage_33, output=""}
    False {from=MaterialExpressionMultiply_16, output=""}
    Value {from=MaterialExpressionNamedRerouteUsage_19, output=""}
   ]
   properties {}
  }
  MaterialExpressionStep_3 {
   class=MaterialExpressionStep
   inputs [
    Y {from=MaterialExpressionMultiply_20, output=""}
    X {from="", output=""}
   ]
   properties {constX=0.5, constY=0.5}
  }
  MaterialExpressionSubtract_0 {
   class=MaterialExpressionSubtract
   inputs [
    A {from=MaterialExpressionTime_0}
    B {from=MaterialExpressionScalarParameter_8}
   ]
  }
  MaterialExpressionSubtract_1 {
   class=MaterialExpressionSubtract
   inputs [
    A {from=MaterialExpressionMultiply_2}
    B {from=MaterialExpressionComponentMask_1}
   ]
  }
  MaterialExpressionSubtract_10 {
   class=MaterialExpressionSubtract
   inputs [
    A {from=MaterialExpressionReroute_34}
    B {from=MaterialExpressionMultiply_40}
   ]
  }
  MaterialExpressionSubtract_11 {
   class=MaterialExpressionSubtract
   inputs [
    A {from=MaterialExpressionLength_0}
    B {from=MaterialExpressionScalarParameter_13}
   ]
  }
  MaterialExpressionSubtract_12 {
   class=MaterialExpressionSubtract
   inputs [
    A {from=MaterialExpressionVertexInterpolator_4, output=PS}
    B {from=MaterialExpressionTransformPosition_0}
   ]
  }
  MaterialExpressionSubtract_13 {
   class=MaterialExpressionSubtract
   inputs [
    A {from=MaterialExpressionObjectPositionWS_13}
    B {from=MaterialExpressionWorldPosition_13, output=XYZ}
   ]
  }
  MaterialExpressionSubtract_14 {
   class=MaterialExpressionSubtract
   inputs [
    A {from=MaterialExpressionComponentMask_6}
    B {from=MaterialExpressionScalarParameter_27}
   ]
  }
  MaterialExpressionSubtract_15 {
   class=MaterialExpressionSubtract
   inputs [
    A {from=MaterialExpressionWorldPosition_2, output=XYZ}
    B {from=MaterialExpressionCameraPositionWS_3}
   ]
  }
  MaterialExpressionSubtract_16 {
   class=MaterialExpressionSubtract
   inputs [
    A {from=MaterialExpressionNamedRerouteUsage_12}
    B {from=MaterialExpressionVectorParameter_6, output=RGB}
   ]
  }
  MaterialExpressionSubtract_17 {
   class=MaterialExpressionSubtract
   inputs [
    A {from=MaterialExpressionAdd_8}
    B {from=""}
   ]
   properties {constB=2}
  }
  MaterialExpressionSubtract_18 {
   class=MaterialExpressionSubtract
   inputs [
    A {from=MaterialExpressionSubtract_17}
    B {from=MaterialExpressionFrac_0}
   ]
  }
  MaterialExpressionSubtract_19 {
   class=MaterialExpressionSubtract
   inputs [
    A {from=MaterialExpressionNamedRerouteUsage_14}
    B {from=MaterialExpressionNamedRerouteDeclaration_11}
   ]
  }
  MaterialExpressionSubtract_2 {
   class=MaterialExpressionSubtract
   inputs [
    A {from=MaterialExpressionComponentMask_7}
    B {from=""}
   ]
  }
  MaterialExpressionSubtract_20 {
   class=MaterialExpressionSubtract
   inputs [
    A {from=MaterialExpressionMultiply_75}
    B {from=MaterialExpressionComponentMask_16}
   ]
  }
  MaterialExpressionSubtract_21 {
   class=MaterialExpressionSubtract
   inputs [
    A {from=MaterialExpressionNamedRerouteDeclaration_13}
    B {from=MaterialExpressionScalarParameter_12}
   ]
  }
  MaterialExpressionSubtract_22 {
   class=MaterialExpressionSubtract
   inputs [
    A {from=MaterialExpressionNamedRerouteUsage_16}
    B {from=MaterialExpressionNamedRerouteUsage_25}
   ]
  }
  MaterialExpressionSubtract_23 {
   class=MaterialExpressionSubtract
   inputs [
    A {from=MaterialExpressionNamedRerouteUsage_17}
    B {from=MaterialExpressionNamedRerouteUsage_26}
   ]
  }
  MaterialExpressionSubtract_24 {
   class=MaterialExpressionSubtract
   inputs [
    A {from=MaterialExpressionMaterialFunctionCall_21, output=Ray Entry Position}
    B {from=MaterialExpressionNamedRerouteDeclaration_16}
   ]
  }
  MaterialExpressionSubtract_25 {
   class=MaterialExpressionSubtract
   inputs [
    A {from=MaterialExpressionDotProduct_10}
    B {from=MaterialExpressionVectorParameter_16, output=R}
   ]
  }
  MaterialExpressionSubtract_26 {
   class=MaterialExpressionSubtract
   inputs [
    A {from=MaterialExpressionVectorParameter_16, output=G}
    B {from=MaterialExpressionVectorParameter_16, output=G}
   ]
  }
  MaterialExpressionSubtract_27 {
   class=MaterialExpressionSubtract
   inputs [
    A {from=MaterialExpressionWorldPosition_3, output=XYZ}
    B {from=MaterialExpressionNamedRerouteUsage_15}
   ]
  }
  MaterialExpressionSubtract_3 {
   class=MaterialExpressionSubtract
   inputs [
    A {from=MaterialExpressionAppendVector_15}
    B {from=MaterialExpressionAppendVector_16}
   ]
  }
  MaterialExpressionSubtract_4 {
   class=MaterialExpressionSubtract
   inputs [
    A {from=MaterialExpressionNamedRerouteUsage_2}
    B {from=MaterialExpressionVectorParameter_4, output=RGB}
   ]
  }
  MaterialExpressionSubtract_5 {
   class=MaterialExpressionSubtract
   inputs [
    A {from=MaterialExpressionComponentMask_4}
    B {from=""}
   ]
   properties {constB=0.376}
  }
  MaterialExpressionSubtract_6 {
   class=MaterialExpressionSubtract
   inputs [
    A {from=MaterialExpressionMaterialFunctionCall_3, output=Result}
    B {from=MaterialExpressionReroute_4}
   ]
  }
  MaterialExpressionSubtract_7 {
   class=MaterialExpressionSubtract
   inputs [
    A {from=MaterialExpressionMaterialFunctionCall_26, output=R}
    B {from=MaterialExpressionMaterialFunctionCall_30, output=Mask_Top}
   ]
  }
  MaterialExpressionSubtract_8 {
   class=MaterialExpressionSubtract
   inputs [
    A {from=MaterialExpressionAdd_24}
    B {from=MaterialExpressionMultiply_22}
   ]
  }
  MaterialExpressionSubtract_9 {
   class=MaterialExpressionSubtract
   inputs [
    A {from=MaterialExpressionScalarParameter_62}
    B {from=MaterialExpressionComponentMask_9}
   ]
  }
  MaterialExpressionTextureObject_0 {
   class=MaterialExpressionTextureObject
   inputs []
   properties {samplerType=SAMPLERTYPE_Masks, texture=/Game/Characters/Player/Male/Large/Bodies/M_LRG_Slurp_Monster/Textures/T_SlurpBubble}
  }
  MaterialExpressionTextureObject_4 {
   class=MaterialExpressionTextureObject
   inputs []
   properties {samplerType=SAMPLERTYPE_Normal, texture=/Game/Characters/Player/Male/Large/Bodies/M_LRG_Slurp_Monster/Textures/T_SlurpBubble_N}
  }
  MaterialExpressionTextureSample_0 {
   class=MaterialExpressionTextureSample
   inputs [
    UVs {from=MaterialExpressionPanner_2}
    Tex {from=MaterialExpressionReroute_26}
    Apply View MipBias {from=""}
   ]
   properties {samplerType=SAMPLERTYPE_Normal}
  }
  MaterialExpressionTextureSample_1 {
   class=MaterialExpressionTextureSample
   inputs [
    UVs {from=MaterialExpressionPanner_0}
    Tex {from=MaterialExpressionReroute_26}
    Apply View MipBias {from=""}
   ]
   properties {samplerType=SAMPLERTYPE_Normal}
  }
  MaterialExpressionTextureSample_8 {
   class=MaterialExpressionTextureSample
   inputs [
    UVs {from=MaterialExpressionPanner_9}
    Tex {from=MaterialExpressionReroute_26}
    Apply View MipBias {from=""}
   ]
   properties {samplerType=SAMPLERTYPE_Normal}
  }
  MaterialExpressionTime_0 {class=MaterialExpressionTime}
  MaterialExpressionTime_3 {class=MaterialExpressionTime}
  MaterialExpressionTime_4 {class=MaterialExpressionTime}
  MaterialExpressionTransformPosition_0 {
   class=MaterialExpressionTransformPosition
   inputs [None {from=MaterialExpressionCameraPositionWS_0, output=""}]
   properties {transformSourceType=TRANSFORMPOSSOURCE_World}
  }
  MaterialExpressionTransform_0 {
   class=MaterialExpressionTransform
   inputs [None {from=MaterialExpressionComponentMask_8}]
   properties {transformSourceType=TRANSFORMSOURCE_World, transformType=TRANSFORM_Local}
  }
  MaterialExpressionTransform_1 {
   class=MaterialExpressionTransform
   inputs [None {from=MaterialExpressionMaterialFunctionCall_9, output=Result}]
   properties {transformSourceType=TRANSFORMSOURCE_World, transformType=TRANSFORM_Local}
  }
  MaterialExpressionTransform_2 {class=MaterialExpressionTransform, inputs [None {from=MaterialExpressionReroute_28}], properties {transformSourceType=TRANSFORMSOURCE_Local}}
  MaterialExpressionTransform_4 {
   class=MaterialExpressionTransform
   inputs [None {from=MaterialExpressionNormalize_8}]
   properties {transformSourceType=TRANSFORMSOURCE_World, transformType=TRANSFORM_Local}
  }
  MaterialExpressionTransform_8 {
   class=MaterialExpressionTransform
   inputs [None {from=MaterialExpressionConstant3Vector_7}]
   properties {transformSourceType=TRANSFORMSOURCE_Local}
  }
  MaterialExpressionVectorParameter_0 {class=MaterialExpressionVectorParameter, properties {parameterName=BackHeadColor, defaultValue=<rgba 0 0.1158 0.3021 1>}}
  MaterialExpressionVectorParameter_1 {class=MaterialExpressionVectorParameter, properties {parameterName=MouthColor, defaultValue=<rgba 0 0.1714 0.4072 1>}}
  MaterialExpressionVectorParameter_10 {class=MaterialExpressionVectorParameter, properties {parameterName=Disc Scale, defaultValue=<rgba 0.1 0.1 0 1>}}
  MaterialExpressionVectorParameter_12 {class=MaterialExpressionVectorParameter, properties {parameterName=SHMax, defaultValue=<rgba 1.25 0.13 0.18 0.13>}}
  MaterialExpressionVectorParameter_13 {class=MaterialExpressionVectorParameter, properties {parameterName=SHMin, defaultValue=<rgba 1 -0.16 -0.17 -0.15>}}
  MaterialExpressionVectorParameter_14 {class=MaterialExpressionVectorParameter, properties {parameterName=Disc Pos, defaultValue=<rgba 0 0 0 1>}}
  MaterialExpressionVectorParameter_15 {class=MaterialExpressionVectorParameter, properties {parameterName=Disc Scale, defaultValue=<rgba 0.1 0.1 0 1>}}
  MaterialExpressionVectorParameter_16 {class=MaterialExpressionVectorParameter, properties {parameterName=Disc Z Min - Max, defaultValue=<rgba 0 0 0 1>}}
  MaterialExpressionVectorParameter_17 {class=MaterialExpressionVectorParameter, properties {group=KneelReedy, parameterName=Clipping Mask Direction, defaultValue=<rgba 0 0 1 0>}}
  MaterialExpressionVectorParameter_18 {class=MaterialExpressionVectorParameter, properties {group=KneelReedy, parameterName=Clipping Pos, defaultValue=<rgba 0 0 -50 0>}}
  MaterialExpressionVectorParameter_19 {class=MaterialExpressionVectorParameter, properties {parameterName=ChonkyPosOffset, defaultValue=<rgba 0 0 0 1>}}
  MaterialExpressionVectorParameter_2 {class=MaterialExpressionVectorParameter, properties {parameterName=OuterColor_Eyes, defaultValue=<rgba 0 0.3177 0.2554 1>}}
  MaterialExpressionVectorParameter_20 {class=MaterialExpressionVectorParameter, properties {parameterName=Disc Direction Offset, defaultValue=<rgba 0 0 0 1>}}
  MaterialExpressionVectorParameter_21 {class=MaterialExpressionVectorParameter, properties {parameterName=XY Scale, defaultValue=<rgba 1 1 0 1>}}
  MaterialExpressionVectorParameter_3 {class=MaterialExpressionVectorParameter, properties {parameterName=InnerColor_Eyes, defaultValue=<rgba 0 0 0 1>}}
  MaterialExpressionVectorParameter_4 {class=MaterialExpressionVectorParameter, properties {parameterName=ClothingMaskCylinder_Center, defaultValue=<rgba 0 0 1 1>}}
  MaterialExpressionVectorParameter_5 {class=MaterialExpressionVectorParameter, properties {group=KneelReedy, parameterName=Disc Direction, defaultValue=<rgba 0 5 1 0>}}
  MaterialExpressionVectorParameter_6 {class=MaterialExpressionVectorParameter, properties {parameterName=CylinderOrigin, defaultValue=<rgba 0 0 0 1>}}
  MaterialExpressionVectorParameter_7 {class=MaterialExpressionVectorParameter, properties {parameterName=CylinderAxis, defaultValue=<rgba 0 0 1 1>}}
  MaterialExpressionVectorParameter_8 {class=MaterialExpressionVectorParameter, properties {parameterName=OuterColor, defaultValue=<rgba 0 0.8802 0.5868 1>}}
  MaterialExpressionVectorParameter_9 {class=MaterialExpressionVectorParameter, properties {parameterName=InnerColor, defaultValue=<rgba 0 0.2918 0.2346 1>}}
  MaterialExpressionVertexColor_0 {class=MaterialExpressionVertexColor, inputs [], properties {}}
  MaterialExpressionVertexInterpolator_2 {class=MaterialExpressionVertexInterpolator, inputs [VS {from=MaterialExpressionLocalPosition_2, output=XYZ}], properties {}}
  MaterialExpressionVertexInterpolator_4 {class=MaterialExpressionVertexInterpolator, inputs [VS {from=MaterialExpressionPreSkinnedPosition_1, output=""}], properties {}}
  MaterialExpressionVertexInterpolator_5 {class=MaterialExpressionVertexInterpolator, inputs [VS {from=MaterialExpressionPreSkinnedNormal_2, output=""}], properties {}}
  MaterialExpressionVertexNormalWS_0 {class=MaterialExpressionVertexNormalWS, inputs [], properties {}}
  MaterialExpressionVertexNormalWS_5 {class=MaterialExpressionVertexNormalWS, inputs [], properties {}}
  MaterialExpressionWorldPosition_0 {class=MaterialExpressionWorldPosition, inputs [], properties {worldPositionShaderOffset=WPT_ExcludeAllShaderOffsets}}
  MaterialExpressionWorldPosition_1 {class=MaterialExpressionWorldPosition, inputs [], properties {}}
  MaterialExpressionWorldPosition_13 {class=MaterialExpressionWorldPosition, inputs [], properties {}}
  MaterialExpressionWorldPosition_2 {class=MaterialExpressionWorldPosition, inputs [], properties {}}
  MaterialExpressionWorldPosition_3 {class=MaterialExpressionWorldPosition, inputs [], properties {worldPositionShaderOffset=WPT_ExcludeAllShaderOffsets}}
 }
 properties {
  bDisableDepthTest=true
  bUseMaterialAttributes=true
  bUsedWithMeshDeformer=true
  bUsedWithSkeletalMesh=true
  blendMode=BLEND_Translucent
  pixelDepthOffsetMode=PDOM_Legacy
  translucencyLightingMode=TLM_VolumetricDirectional
  usedShadingModels=Default Lit
 }
}
