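# Post-process AnimBP for the BandSoda quadruped cosmetic companion, adding facial curve remapping and secondary motion.
# An alpha-weighted additive pose feeds a curve remap from the attached parent mesh, then the pose asset.
# A weighted seven-bone constraint drives the collar mask bone, then chained AnimDynamics springs jiggle ears, tongue, blobs and tail.
# DynamicsAlpha scales only ears and tongue; IsBPCompanion updates each frame but is unread here.
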AnimBlueprint /CosmeticCompanions/Assets/Quadruped/BandSoda/Meshes/BandSoda_Quadruped_Post_AnimBP
target_skeleton=&/BandSoda_Quadruped_Skeleton
graphs [
 AnimGraph {nodes [
   n1 = Remap Curves From Mesh!(
    SourcePose:Pose Link Structure=Use cached pose 'StartCache'()
    CurveExpressionsDataAsset:Curve Expressions Data Asset Object Reference=&/BandSoda_Remap
    SourceMeshComponent:Skeletal Mesh Component Weak object=Get mesh component() -> (mesh component:Skeletal Mesh Component Object Reference)
   ) -> (Pose:Pose Link Structure) {is=title AnimGraphNode_RemapCurvesFromMesh, bUseAttachedParent=true, expressionSource=DataAsset}
   n2 = "Constraint - deform_collar_mat_mask (PPPPPPP)"!(
    ComponentPose=Local To Component(LocalPose=BandSoda_Quadruped_PoseAsset!(SourcePose:Pose Link Structure=@n1) -> (Pose:Pose Link Structure) {is=title AnimGraphNode_PoseBlendNode, poseAsset=&/BandSoda_Quadruped_PoseAsset})
    Alpha=1
    ConstraintWeights_0=0.6795
    ConstraintWeights_1:float=0.2396
    ConstraintWeights_2:float=0.0615
    ConstraintWeights_3:float=0.0163
    ConstraintWeights_4:float=0.0018
    ConstraintWeights_5:float=0.0008
    ConstraintWeights_6:float=0.0005
   ) -> (Pose) {
    boneToModify {boneName=deform_collar_mat_mask}
    constraintSetup [
     {targetBone {boneName=neck_02}}
     {targetBone {boneName=spine_02}}
     {targetBone {boneName=head}}
     {targetBone {boneName=spine_01}}
     {targetBone {boneName=neck_01}}
     {targetBone {boneName=hips}}
     {targetBone {boneName=spine_03}}
    ]
    constraintWeights [1,1,1,1,1,1,1]
   }
   n3 = Get DynamicsAlpha() -> (DynamicsAlpha:float)
   n4 = Anim Dynamics - Bone: ear_03_l(ComponentPose=@n2, Alpha=@n3, ExternalForce="0,0,0") {
    angularDampingOverride=1
    angularSpringConstant=20000
    bAngularSpring=true
    bOverrideAngularDamping=true
    bUseGravityOverride=true
    gravityScale=0.35
    simulationSpace=RootRelative
    boundBone {boneName=ear_03_l}
    componentLinearAccScale=<1 1 1>
    componentLinearVelScale=<1 1 1>
    gravityOverride {z=-980}
    physicsBodyDefinitions [{
      boundBone {boneName=ear_03_l}
      boxExtents=<20 30 30>
      constraintSetup {angularLimitsMax=<15 15 15>, angularLimitsMin=<-15 -15 -15>, angularTarget=<1 0 0>}
      localJointOffset=<4 4 0>
     }]
   }
   n5 = Anim Dynamics - Bone: ear_04_l(ComponentPose=@n4, Alpha=@n3, ExternalForce="0,0,0") {
    angularDampingOverride=1
    angularSpringConstant=20000
    bAngularSpring=true
    bOverrideAngularDamping=true
    bUseGravityOverride=true
    gravityScale=0.35
    simulationSpace=RootRelative
    boundBone {boneName=ear_04_l}
    componentLinearAccScale=<1 1 1>
    componentLinearVelScale=<1 1 1>
    gravityOverride {z=-980}
    physicsBodyDefinitions [{
      boundBone {boneName=ear_04_l}
      boxExtents=<20 30 30>
      constraintSetup {angularLimitsMax=<15 15 15>, angularLimitsMin=<-15 -15 -15>, angularTarget=<1 0 0>}
      localJointOffset=<4 4 0>
     }]
   }
   n6 = Anim Dynamics - Bone: ear_03_r(ComponentPose=@n5, Alpha=@n3, ExternalForce="0,0,0") {
    angularDampingOverride=1
    angularSpringConstant=20000
    bAngularSpring=true
    bOverrideAngularDamping=true
    bUseGravityOverride=true
    gravityScale=0.35
    simulationSpace=RootRelative
    boundBone {boneName=ear_03_r}
    componentLinearAccScale=<1 1 1>
    componentLinearVelScale=<1 1 1>
    gravityOverride {z=-980}
    physicsBodyDefinitions [{
      boundBone {boneName=ear_03_r}
      boxExtents=<20 30 30>
      constraintSetup {angularLimitsMax=<15 15 15>, angularLimitsMin=<-15 -15 -15>, angularTarget=<1 0 0>}
      localJointOffset=<4 -4 0>
     }]
   }
   n7 = Anim Dynamics - Bone: ear_04_r(ComponentPose=@n6, Alpha=@n3, ExternalForce="0,0,0") {
    angularDampingOverride=1
    angularSpringConstant=20000
    bAngularSpring=true
    bOverrideAngularDamping=true
    bUseGravityOverride=true
    gravityScale=0.35
    simulationSpace=RootRelative
    boundBone {boneName=ear_04_r}
    componentLinearAccScale=<1 1 1>
    componentLinearVelScale=<1 1 1>
    gravityOverride {z=-980}
    physicsBodyDefinitions [{
      boundBone {boneName=ear_04_r}
      boxExtents=<20 30 30>
      constraintSetup {angularLimitsMax=<15 15 15>, angularLimitsMin=<-15 -15 -15>, angularTarget=<1 0 0>}
      localJointOffset=<4 -4 0>
     }]
   }
   n8 = Anim Dynamics - Bone: tongue_01(ComponentPose=@n7, Alpha=@n3, ExternalForce="0,0,0") {
    angularDampingOverride=1
    angularSpringConstant=20000
    bAngularSpring=true
    bOverrideAngularDamping=true
    bUseGravityOverride=true
    gravityScale=0.35
    simulationSpace=RootRelative
    boundBone {boneName=tongue_01}
    componentLinearAccScale=<1 1 1>
    componentLinearVelScale=<1 1 1>
    gravityOverride {z=-980}
    physicsBodyDefinitions [{
      boundBone {boneName=tongue_01}
      boxExtents=<20 30 30>
      constraintSetup {angularLimitsMax=<5 0 5>, angularLimitsMin=<-5 0 -5>, angularTarget=<1 0 0>}
      localJointOffset=<-4 0 0>
     }]
   }
   n9 = Anim Dynamics - Bone: tongue_02(ComponentPose=@n8, Alpha=@n3, ExternalForce="0,0,0") {
    angularDampingOverride=1
    angularSpringConstant=20000
    bAngularSpring=true
    bOverrideAngularDamping=true
    bUseGravityOverride=true
    gravityScale=0.35
    simulationSpace=RootRelative
    boundBone {boneName=tongue_02}
    componentLinearAccScale=<1 1 1>
    componentLinearVelScale=<1 1 1>
    gravityOverride {z=-980}
    physicsBodyDefinitions [{
      boundBone {boneName=tongue_02}
      boxExtents=<20 30 30>
      constraintSetup {angularLimitsMax=<10 10 10>, angularLimitsMin=<-10 -10 -10>, angularTarget=<1 0 0>}
      localJointOffset=<-4 0 0>
     }]
   }
   n10 = Anim Dynamics - Bone: tongue_03(ComponentPose=@n9, Alpha=@n3, ExternalForce="0,0,0") {
    angularDampingOverride=1
    angularSpringConstant=20000
    bAngularSpring=true
    bOverrideAngularDamping=true
    bUseGravityOverride=true
    gravityScale=0.35
    simulationSpace=RootRelative
    boundBone {boneName=tongue_03}
    componentLinearAccScale=<1 1 1>
    componentLinearVelScale=<1 1 1>
    gravityOverride {z=-980}
    physicsBodyDefinitions [{
      boundBone {boneName=tongue_03}
      boxExtents=<20 30 30>
      constraintSetup {angularLimitsMax=<10 10 10>, angularLimitsMin=<-10 -10 -10>, angularTarget=<1 0 0>}
      localJointOffset=<-4 0 0>
     }]
   }
   n11 = Anim Dynamics - Bone: dyn_top_blob(ComponentPose=@n10, Alpha=1, ExternalForce="0,0,0") {
    angularDampingOverride=1
    angularSpringConstant=20000
    bAngularSpring=true
    bOverrideAngularDamping=true
    bUseGravityOverride=true
    gravityScale=0.35
    simulationSpace=RootRelative
    boundBone {boneName=dyn_top_blob}
    componentLinearAccScale=<1 1 1>
    componentLinearVelScale=<1 1 1>
    gravityOverride {z=-980}
    physicsBodyDefinitions [{
      boundBone {boneName=dyn_top_blob}
      boxExtents=<20 30 30>
      constraintSetup {angularLimitsMax=<15 15 15>, angularLimitsMin=<-15 -15 -15>, angularTarget=<1 0 0>}
      localJointOffset=<4 0 0>
     }]
   }
   n12 = Anim Dynamics - Bone: dyn_inner_blob_01(ComponentPose=@n11, Alpha=1, ExternalForce="0,0,0") {
    angularDampingOverride=1
    angularSpringConstant=20000
    bAngularSpring=true
    bOverrideAngularDamping=true
    bUseGravityOverride=true
    gravityScale=0.35
    simulationSpace=RootRelative
    boundBone {boneName=dyn_inner_blob_01}
    componentLinearAccScale=<1 1 1>
    componentLinearVelScale=<1 1 1>
    gravityOverride {z=-980}
    physicsBodyDefinitions [{
      boundBone {boneName=dyn_inner_blob_01}
      boxExtents=<20 30 30>
      constraintSetup {angularLimitsMax=<0 30 30>, angularLimitsMin=<0 -30 -30>, angularTarget=<1 0 0>}
      localJointOffset=<-4 0 0>
     }]
   }
   n13 = Anim Dynamics - Bone: dyn_inner_blob_02(ComponentPose=@n12, Alpha=1, ExternalForce="0,0,0") {
    angularDampingOverride=1
    angularSpringConstant=20000
    bAngularSpring=true
    bOverrideAngularDamping=true
    bUseGravityOverride=true
    gravityScale=0.35
    linearSpringConstant=20000
    simulationSpace=RootRelative
    boundBone {boneName=dyn_inner_blob_02}
    componentLinearAccScale=<1 1 1>
    componentLinearVelScale=<1 1 1>
    gravityOverride {z=-980}
    physicsBodyDefinitions [{
      boundBone {boneName=dyn_inner_blob_02}
      boxExtents=<20 30 30>
      constraintSetup {angularLimitsMax=<40 15 30>, angularLimitsMin=<-40 -15 -30>, angularTarget=<1 0 0>}
      localJointOffset=<4 0 0>
     }]
   }
   n14 = Anim Dynamics - Bone: dyn_inner_blob_03(ComponentPose=@n13, Alpha=1, ExternalForce="0,0,0") {
    angularDampingOverride=0.999
    angularSpringConstant=20000
    bAngularSpring=true
    bOverrideAngularDamping=true
    bUseGravityOverride=true
    gravityScale=0.35
    linearDampingOverride=1
    linearSpringConstant=200
    simulationSpace=RootRelative
    boundBone {boneName=dyn_inner_blob_03}
    componentLinearAccScale=<1 1 1>
    componentLinearVelScale=<1 1 1>
    gravityOverride {z=-980}
    physicsBodyDefinitions [{
      boundBone {boneName=dyn_inner_blob_03}
      boxExtents=<20 30 30>
      constraintSetup {angularLimitsMax=<60 15 45>, angularLimitsMin=<-60 -15 -45>, angularTarget=<1 0 0>}
      localJointOffset=<4 0 0>
     }]
   }
   n15 = Anim Dynamics - Bone: dyn_tail_02(ComponentPose=@n14, Alpha=1, ExternalForce="0,0,0") {
    angularDampingOverride=1
    angularSpringConstant=20000
    bAngularSpring=true
    bOverrideAngularDamping=true
    bUseGravityOverride=true
    gravityScale=0.35
    simulationSpace=RootRelative
    boundBone {boneName=dyn_tail_02}
    componentLinearAccScale=<1 1 1>
    componentLinearVelScale=<1 1 1>
    gravityOverride {z=-980}
    physicsBodyDefinitions [{
      boundBone {boneName=dyn_tail_02}
      boxExtents=<20 30 30>
      constraintSetup {angularLimitsMax=<0 15 20>, angularLimitsMin=<0 -30 -20>, angularTarget=<1 0 0>}
      localJointOffset=<4 0 0>
     }]
   }
   Output Pose(Result=Component To Local(ComponentPose=@n15))
   n16 = Get AdditivePoseAnim() -> (AdditivePoseAnim:Anim Sequence Object Reference)
   Save cached pose 'StartCache'(Pose=Apply Additive(
     Base=Input Pose()
     Additive=Sequence Player!(Sequence:Anim Sequence Base Object Reference=@n16) -> (Pose)
     Alpha=Get CosmeticPoseAlpha() -> (CosmeticPoseAlpha:float)
    ) {alphaBoolBlend {blendOption=Linear}}) {cacheName=StartCache}
  ]}
 EventGraph {nodes [
   Event BlueprintInitializeAnimation()
   > Set mesh component(mesh component:Skeletal Mesh Component Object Reference=GetOwningComponent()) -> (Output_Get:Skeletal Mesh Component Object Reference)
   Event BlueprintUpdateAnimation()
   > Set IsBPCompanion(IsBPCompanion:bool=IsA!(
     Object:Object Reference=GetOwningActor()
     ObjectClass:Object Class Reference=/CosmeticCompanions/Assets/Shared/Blueprint/BP_Companion.BP_Companion_C
    ) -> (ReturnValue:bool)) -> (Output_Get:bool)
  ]}
]
variables [
 IsBPCompanion {type=boolean}
 Mesh Component {type=SkeletalMeshComponent}
 PoseAsset {type=PoseAsset, default=/CosmeticCompanions/Automation_Template/companion_pose_asset_template}
]
